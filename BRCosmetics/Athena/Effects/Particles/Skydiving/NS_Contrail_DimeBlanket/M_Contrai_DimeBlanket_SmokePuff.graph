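# Translucent Niagara sprite material for smoke puffs in the DimeBlanket skydiving contrail effect, with volumetric per-vertex directional lighting.
# A dynamic Animation parameter drives motion-vector 4x4 flipbooks for depth, alpha, emissive and normals.
# Opacity logic combines noise dissolve, depth fade, particle alpha and optional camera-distance and Z-axis masks.
# Color logic offers sphere-blended biased normals, time-of-day tinting with sun-based shading, a fake point light, and noise-eroded, curve-atlas HDR emissive.
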Material /BRCosmetics/Athena/Effects/Particles/Skydiving/NS_Contrail_DimeBlanket/M_Contrai_DimeBlanket_SmokePuff
graph {
 properties {
  bAutomaticallySetUsageInEditor=false
  bUsedWithNiagaraSprites=true
  blendMode=BLEND_Translucent
  pixelDepthOffsetMode=PDOM_Legacy
  translucencyLightingMode=TLM_VolumetricPerVertexDirectional
  translucentSelfShadowDensityScale=1
  usedShadingModels=Default Lit
  translucentMultipleScatteringExtinction=<rgba 1 1 1 1>
 }
 nodes [
  Constant()
  Constant() {r=1}
  MaterialFunctionCall(Emissive=NamedRerouteUsage()) {materialFunction=&MF_FX_EmissiveAdjustments}
  n1 = DynamicParameter() {paramNames [Animation, Unused, Emissive, NormalMultiplier]}
  n2 = Constant() {r=4}
  n3 = TextureObject() {texture=/Game/Effects/Niagara/MasterMaterials/Smoke/Textures/NEW/RT4x4_1_Tex}
  n4 = TextureObjectParameter() {parameterName=MotionVector, samplerType=SAMPLERTYPE_LinearColor, texture=/Game/Effects/Niagara/MasterMaterials/Smoke/Textures/Vector_OF}
  n5 = ScalarParameter() {defaultValue=0.344, group=Motion Vectors, parameterName=MotionVector}
  NamedRerouteDeclaration(None=ComponentMask(None=MaterialFunctionCall!(
     "Animation  Phase (0-1)"=@n1.Animation
     Number of Columns X=@n2
     Number of Rows Y=@n2
     Texture=@n3
     UVs
     Motion Vector Texture=@n4
     Phase to Debug
     Motion Vector Intensity=@n5
     Double Apply Motion Vectors
     Custom UV Input
     Custom UV2 Input
     Use Custom UVs
    ) {materialFunction=/Engine/Functions/Engine_MaterialFunctions02/Texturing/FlipBook_MotionVectors}.Result) {g=true}) {name=Depth_Flipbook, nodeColor=<rgba 0 0.8594 1 1>}
  n6 = TextureObject() {texture=/Game/Effects/Niagara/MasterMaterials/Smoke/Textures/NEW/RT4x4_4_Tex}
  NamedRerouteDeclaration(None=ComponentMask(None=MaterialFunctionCall(
     "Animation  Phase (0-1)"=@n1.Animation
     Number of Columns X=@n2
     Number of Rows Y=@n2
     Texture=@n6
     Motion Vector Texture=@n4
     Motion Vector Intensity=@n5
    ) {materialFunction=&FlipBook_MotionVectors}.Result) {g=true}) {name=Alpha_Flipbook, nodeColor=<rgba 1 0 0.3516 1>}
  n7 = ComponentMask(None=NamedRerouteUsage()) {g=true, r=true}
  n8 = DynamicParameter() {
   parameterIndex=2
   defaultValue=<rgba 0 0 0 1>
   paramNames [Smoke - DissolveLerpNoise, Smoke - DissolveNoiseOffsetX, Smoke - DissolveNoiseErosionPower, Smoke - DissolveNoiseScale]
  }
  n9 = Constant()
  n10 = ScalarParameter() {group=Normal, parameterName="DissolveDetail  Random Y Offset Value", sliderMax=2}
  n11 = Add(A=Add(A=TextureCoordinate(), B=@n8.Smoke - DissolveNoiseOffsetX), B=Multiply(A=ParticleRandom(), B=AppendVector(A=@n9, B=@n10)))
  n12 = TextureSampleParameter2D(UVs=Add(A=Multiply(A=@n7) {constB=-0.2}, B=Multiply(A=@n11, B=@n8.Smoke - DissolveNoiseScale) {desc=Scale})) {parameterName=DissolveDetailTexture, samplerType=SAMPLERTYPE_LinearColor, texture=/Game/Global/Textures/Default/Noise/T_Noise01_Inverse}
  n13 = DynamicParameter() {
   parameterIndex=2
   defaultValue=<rgba 0 0 0 1>
   paramNames [Smoke - DissolveLerpNoise, Smoke - DissolveNoiseOffsetX, Smoke - DissolveNoiseErosionPower, Smoke - DissolveNoiseScale]
  }
  NamedRerouteDeclaration(None=Clamp(None=Power(Base=@n12.R, Exp=@n13.Smoke - DissolveNoiseErosionPower))) {name=DissolveDetail, nodeColor=<rgba 0.3828 0 1 1>}
  n14 = ScalarParameter() {defaultValue=100, parameterName=DepthFade_Low}
  n15 = ScalarParameter() {defaultValue=10, parameterName=DepthFade_high}
  n16 = Multiply(A=NamedRerouteUsage(), B=NamedRerouteUsage())
  NamedRerouteDeclaration(None=DepthFade(FadeDistance=LinearInterpolate(A=@n14, B=@n15, Alpha=@n16))) {name=DepthFade, nodeColor=<rgba 0.875 1 0 1>}
  n17 = Subtract(A=NamedRerouteUsage(), B=Constant() {r=-1})
  n18 = ScalarParameter() {parameterName=Contrast}
  n19 = NamedRerouteUsage()
  n20 = NamedRerouteUsage()
  n21 = Multiply(
   A=Multiply(
    A=MaterialFunctionCall(In=@n17, Contrast=@n18) {materialFunction=&CheapContrast}.Result
    B=Multiply(A=@n19, B=@n20)
   )
   B=ParticleColor().A
  )
  n22 = DynamicParameter() {
   parameterIndex=2
   defaultValue=<rgba 0 0 0 1>
   paramNames [Smoke - DissolveLerpNoise, Smoke - DissolveNoiseOffsetX, Smoke - DissolveNoiseErosionPower, Smoke - DissolveNoiseScale]
  }
  n23 = LinearInterpolate(A=@n21, B=Multiply(A=@n21, B=NamedRerouteUsage()), Alpha=Clamp(None=@n22.Smoke - DissolveLerpNoise))
  n24 = SphereMask(
   A=PixelDepth().R
   B=Constant()
   Radius=ScalarParameter() {defaultValue=400, parameterName=NearCameraFadeRadius}
   Hardness=ScalarParameter() {parameterName=NearCameraFadeHardness}
  )
  n25 = StaticSwitchParameter(True=Multiply(A=@n23, B=OneMinus(None=@n24)), False=@n23) {parameterName=IsUsingCameraFade}
  n26 = ParticlePositionWS()
  n27 = VectorParameter() {parameterName=ParticleZaxisOpacityMask_ParticlePositionOffset, defaultValue=<rgba 0 0 -15 1>}
  n28 = ComponentMask(None=Subtract(A=Add(A=@n26, B=@n27.RGB), B=WorldPosition().XYZ)) {b=true}
  n29 = Saturate(None=Divide(A=Multiply(A=@n28) {constB=-1}, B=ScalarParameter() {defaultValue=20, parameterName=ParticleZaxisOpacityMask_Blur}))
  NamedRerouteDeclaration(None=StaticSwitchParameter(True=Multiply(A=@n25, B=@n29), False=@n25) {parameterName=ParticleZaxisOpacityMask}) {name=Final Alpha, nodeColor=<rgba 1 0 0.2578 1>}
  n30 = VectorParameter() {group=SphereMask, parameterName=SphereMask_Color, defaultValue=<rgba 2 0.102 0 0>}
  n31 = DynamicParameter() {parameterIndex=3, paramNames [Red, Blue, Green, NormalMultiplier]}
  n32 = LinearInterpolate(
   B=Multiply(A=NamedRerouteUsage(), B=Multiply(A=@n30.RGB, B=@n31.Red))
   Alpha=SphereMask(
    A=WorldPosition().XYZ
    B=ActorPositionWS()
    Radius=DynamicParameter() {parameterIndex=3, paramNames [Red, Blue, Green, NormalMultiplier]}.Blue
    Hardness=ScalarParameter() {group=SphereMask, parameterName=Hardness}
   )
  )
  n33 = DotProduct(A=Transform(None=NamedRerouteUsage()) {transformType=TRANSFORM_Instance}, B=Constant3Vector() {constant=<rgba 0 0 -0.1 0>})
  n34 = Power(Base=Saturate(None=Add(A=@n33) {constA=-1.088, constB=0.1})) {constExponent=1.5}
  NamedRerouteDeclaration(None=StaticSwitchParameter(True=Multiply(A=@n32, B=@n34), False=@n32) {parameterName=USE_Normal for FakePointLit}) {name=FakePointLight, nodeColor=<rgba 1 0.2812 0 1>}
  n35 = VectorParameter() {parameterName=Color, defaultValue=<rgba 0.3958 0.3987 0.4 0>}
  n36 = DynamicParameter() {parameterIndex=3, paramNames [Red, Blue, Green, NormalMultiplier]}
  n37 = MaterialFunctionCall(R=@n36.Red, G=@n36.Red, B=@n36.Red) {materialFunction=&AppendMany}
  n38 = ParticleColor()
  n39 = VectorParameter() {parameterName=ColorBlend, defaultValue=<rgba 0 1 0 1>}
  n40 = NamedRerouteUsage()
  n41 = SmoothStep(Min=@n39.R, Max=@n39.R, Value=@n40)
  n42 = NamedRerouteDeclaration(None=LinearInterpolate(A=@n35.RGB, B=LinearInterpolate(A=@n37.RGB, B=@n38.RGB, Alpha=@n41), Alpha=ScalarParameter() {parameterName=UseParticleColor})) {name=Final_Base Color, nodeColor=<rgba 1 0.7266 0 1>}
  n43 = Multiply(
   A=MaterialFunctionCall(
    Color_Morning=VectorParameter() {parameterName=Color Morning Multiplier, defaultValue=<rgba 0.9896 0.268 0.4929 0>}.RGB
    Color_Day=VectorParameter() {parameterName=Color Day Multiplier, defaultValue=<rgba 1 1 1 0>}.RGB
    Color_Evening=VectorParameter() {parameterName=Color Evening Multiplier, defaultValue=<rgba 0.599 0.2901 0.3358 0>}.RGB
    Color_Night=VectorParameter() {parameterName=Color Night Multiplier, defaultValue=<rgba 0.3498 0.4864 0.6458 0>}.RGB
   ) {materialFunction=&Color_Time_of_Day}.Result
   B=@n42
  )
  n44 = Multiply(A=@n43, B=ScalarParameter() {defaultValue=0.7, parameterName=Shadow Multiplier}) {constB=0.74}
  n45 = Transform(None=NamedRerouteUsage())
  n46 = CollectionParameter() {
   collection=&FortniteMaterialParameters
   parameterName=SunlightDirectionalVector
  }
  n47 = ComponentMask(None=@n46) {b=true, g=true, r=true}
  n48 = DotProduct(A=@n45, B=@n47)
  n49 = Add(A=Multiply(A=@n35.RGB, B=ScalarParameter() {parameterName=Emissive}), B=NamedRerouteUsage())
  n50 = StaticSwitchParameter(True=@n42, False=Add(A=LinearInterpolate(A=@n44, B=@n43, Alpha=@n48), B=@n49)) {parameterName=IsFlatUnlit_NoTOD}
  NamedRerouteDeclaration(None=StaticSwitchParameter(True=NamedRerouteUsage(), False=StaticSwitchParameter(True=@n50, False=@n49) {parameterName=IsShaderUnlit}) {parameterName=IsUsingFakePointLightt}) {name=Final_Emissive, nodeColor=<rgba 1 0.0938 0 1>}
  n51 = ParticlePositionWS()
  n52 = WorldPosition()
  NamedRerouteDeclaration(None=LinearInterpolate(A=Subtract(A=@n51, B=@n52.XYZ), Alpha=NamedRerouteUsage()) {constB=0}) {name=WPO Shrink, nodeColor=<rgba 0.4766 0 1 1>}
  n53 = TextureObject() {texture=/Game/Effects/Niagara/MasterMaterials/Smoke/Textures/NEW/RT4x4_5_Tex}
  NamedRerouteDeclaration(None=ComponentMask(None=MaterialFunctionCall(
     "Animation  Phase (0-1)"=@n1.Animation
     Number of Columns X=@n2
     Number of Rows Y=@n2
     Texture=@n53
     Motion Vector Texture=@n4
     Motion Vector Intensity=@n5
    ) {materialFunction=&FlipBook_MotionVectors}.Result) {g=true}) {name=Emissive_Flipbook, nodeColor=<rgba 0 1 0.6953 1>}
  n54 = TextureObjectParameter() {parameterName=NormalTexture, samplerType=SAMPLERTYPE_Normal, texture=/Game/Effects/Niagara/MasterMaterials/Smoke/Textures/NEW/RT4x4_3_Tex}
  NamedRerouteDeclaration(None=ComponentMask(None=MaterialFunctionCall(
     "Animation  Phase (0-1)"=@n1.Animation
     Number of Columns X=@n2
     Number of Rows Y=@n2
     Texture=@n54
     Motion Vector Texture=@n4
     Motion Vector Intensity=@n5
    ) {materialFunction=&FlipBook_MotionVectors}.Result) {b=true, g=true, r=true}) {name=Normal_Flipbook, nodeColor=<rgba 0.9453 0 1 1>}
  n55 = Constant3Vector() {constant=<rgba 0 0 1 0>}
  n56 = NamedRerouteDeclaration(None=TextureSample() {samplerType=SAMPLERTYPE_Normal, texture=/Engine/Functions/Engine_MaterialFunctions02/ExampleContent/Textures/SphereRenderNormalsMap}.RGB) {name=SphereNormal_Texture, nodeColor=<rgba 1 0.3281 0 1>}
  n57 = Multiply(
   A=ScalarParameter() {defaultValue=1, group=Normal, parameterName=SphereNormal}
   B=DynamicParameter() {paramNames [Animation, Unused, Emissive, NormalMultiplier]}.NormalMultiplier
  )
  NamedRerouteDeclaration(None=Normalize(VectorInput=LinearInterpolate(A=@n55, B=@n56, Alpha=@n57))) {name=SphereNormal, nodeColor=<rgba 1 0 0.2109 1>}
  n58 = ScalarParameter() {defaultValue=90, desc=FOV Scale Floor, group=CamFade, parameterName=CamFade FOV Floor, sortPriority=4}
  n59 = Max(A=MaterialFunctionCall() {materialFunction=/Engine/Functions/Engine_MaterialFunctions02/Utility/FOV}.FOV angle in degrees, B=@n58)
  n60 = ScalarParameter() {defaultValue=130, desc=FOV Scale Celing, group=CamFade, parameterName=CamFade FOV Ceil, sortPriority=5}
  n61 = ScalarParameter() {
   defaultValue=1
   desc="Scales INPUT DISTANCE by FOV (before remapping it to near/far) - so larger numbers reduce the effect. Floor defaults to 1.0"
   group=CamFade
   parameterName=CamFade ScaleDistByFOV Floor
   sortPriority=8
  }
  n62 = ScalarParameter() {
   defaultValue=1
   desc="Scales INPUT DISTANCE by FOV (before remapping it to near/far) - so larger numbers reduce the effect. Celing defaults to 1.0 but is meaningful at 2.0"
   group=CamFade
   parameterName=CamFade ScaleDistByFOV Ceiling
   sortPriority=9
  }
  n63 = CameraPositionWS()
  n64 = ParticlePositionWS()
  n65 = Multiply(
   A=MaterialFunctionCall(Input=Min(A=@n59, B=@n60), Input Low=@n58, Input High=@n60, Target Low=@n61, Target High=@n62) {materialFunction=&RemapValueRange}.Result
   B=Length(None=Subtract(A=@n63, B=@n64)) {is=class MaterialExpressionLength}
  )
  n66 = ScalarParameter() {defaultValue=25, desc=Near distance to camera --250, group=CamFade, parameterName=CamFade Near, sliderMax=250, sortPriority=1}
  n67 = ScalarParameter() {defaultValue=25, desc=Far distance to camera -- 300, group=CamFade, parameterName=CamFade Far, sliderMax=300, sortPriority=2}
  NamedRerouteDeclaration(None=Saturate(None=MaterialFunctionCall(Input=@n65, Input Low=@n66, Input High=@n67) {materialFunction=&RemapValue_0-1}.Result)) {name=ClipFade Alpha, nodeColor=<rgba 0 1 0.9447 1>}
  n68 = NamedRerouteUsage()
  n69 = ComponentMask(None=VectorParameter() {group=Normal, parameterName=NormalBias Direction, defaultValue=<rgba 0.5 0.5 0 0>}.RGB) {b=true, g=true, r=true}
  n70 = ScalarParameter() {defaultValue=1, desc=Toggle the Bias direction from Tangent to World Space, group=Normal, parameterName=NormalBias WorldSpace}
  n71 = Add(
   A=@n68
   B=LinearInterpolate(A=@n69, B=Transform(None=@n69) {transformSourceType=TRANSFORMSOURCE_World, transformType=TRANSFORM_Tangent}, Alpha=@n70)
  )
  n72 = ComponentMask(None=VectorParameter() {group=Normal, parameterName=NormalBias Multiplier, defaultValue=<rgba 2 2 1 0>}.RGB) {b=true, g=true, r=true}
  NamedRerouteDeclaration(None=StaticSwitchParameter(True=@n68, False=Multiply(A=@n71, B=@n72)) {group=Normal, parameterName=DontUseNormalBias}) {name=BiasedNormal, nodeColor=<rgba 0.6875 1 0 1>}
  n73 = LinearInterpolate(A=NamedRerouteUsage(), B=NamedRerouteUsage(), Alpha=ScalarParameter() {defaultValue=1, group=Normal, parameterName=NormalTexture})
  NamedRerouteDeclaration(None=MaterialFunctionCall(BaseNormal=Normalize(VectorInput=@n73), AdditionalNormal=NamedRerouteUsage()) {materialFunction=&BlendAngleCorrectedNormals}.Result) {name=Final Normal, nodeColor=<rgba 1 0 0.75 1>}
  n74 = CurveAtlasRowParameter(CurveTime=@n40) {
   atlas=/Game/Effects/Niagara/MasterMaterials/Smoke/Textures/CurveAtlas
   curve=/Game/Effects/Niagara/MasterMaterials/Smoke/Textures/EmissiveCurve
   defaultValue=1
   parameterName=EmissiveCurve
  }
  n75 = Multiply(A=NamedRerouteUsage(), B=@n74)
  n76 = MaterialFunctionCall(
   Hue Shift Percentage=ScalarParameter() {parameterName=HueShift}
   Texture=StaticSwitchParameter(True=Clamp(None=@n75), False=@n74) {parameterName=Multiply HDR Emissive by Base Color}
  ) {materialFunction=&HueShift}
  n77 = MaterialFunctionCall(
   LDR Input=@n76.Result
   HDR Tint=@n76.Result
   HDR Intensity Multiplier=Multiply(
    A=ScalarParameter() {defaultValue=100, parameterName=HDR_Intensity}
    B=DynamicParameter() {paramNames [Animation, Unused, Emissive, NormalMultiplier]}.Emissive
   )
  ) {materialFunction=&DeriveHDRfromLDR}
  n78 = NamedRerouteUsage()
  n79 = TextureCoordinate()
  n80 = DynamicParameter() {
   parameterIndex=1
   defaultValue=<rgba 0 0 1 1>
   paramNames [Emissive - LerpNoise, Emissive - NoiseOffsetX, Emissive - NoiseErosionPower, Emissive - NoiseScale]
  }
  n81 = Add(
   A=Multiply(A=ComponentMask(None=@n78) {g=true, r=true}) {constB=-0.2}
   B=Multiply(A=Add(A=@n79, B=@n80.Emissive - NoiseOffsetX), B=@n80.Emissive - NoiseScale) {desc=Scale}
  )
  n82 = DynamicParameter() {
   parameterIndex=1
   defaultValue=<rgba 0 0 1 1>
   paramNames [Emissive - LerpNoise, Emissive - NoiseOffsetX, Emissive - NoiseErosionPower, Emissive - NoiseScale]
  }
  n83 = Multiply(
   A=@n77.RemappedRange
   B=Power(
    Base=TextureSampleParameter2D(UVs=@n81) {parameterName=DetailTexture, texture=/Game/Global/Textures/Default/Tileable_Patterns/T_Blowing_Smoke_Pattern2_Tile_DISP}.G
    Exp=@n82.Emissive - NoiseErosionPower
   )
  )
  n84 = DynamicParameter() {
   parameterIndex=1
   defaultValue=<rgba 0 0 1 1>
   paramNames [Emissive - LerpNoise, Emissive - NoiseOffsetX, Emissive - NoiseErosionPower, Emissive - NoiseScale]
  }
  n85 = Clamp(None=@n84.Emissive - LerpNoise)
  NamedRerouteDeclaration(None=StaticSwitchParameter(True=LinearInterpolate(A=@n77.RemappedRange, B=@n83, Alpha=@n85), False=Constant()) {parameterName=UseEmissive}) {name=HDR Emissive, nodeColor=<rgba 1 0 0.3047 1>}
  NamedRerouteUsage()
  NamedRerouteUsage()
  NamedRerouteUsage()
  n86 = Constant3Vector()
  n87 = DotProduct(
   A=@n47
   B=MaterialFunctionCall(Normal=@n45, Flatness=Constant() {r=0.7}) {materialFunction=&FlattenNormal}.Result
  )
  n88 = LinearInterpolate(A=ScalarParameter() {defaultValue=2, parameterName=ShadowAlphaMult}, Alpha=Saturate(None=Multiply(A=@n87) {constB=5}))
  n89 = NamedRerouteUsage()
  n90 = NamedRerouteUsage()
  StaticSwitchParameter(True=Multiply(A=LinearInterpolate(A=@n86, B=@n88, Alpha=@n89), B=@n90), False=@n90) {parameterName=ShadowAlphaCompensation}
 ]
}
