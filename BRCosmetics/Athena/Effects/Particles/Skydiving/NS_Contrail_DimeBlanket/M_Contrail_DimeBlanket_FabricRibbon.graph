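# Two-sided, lit fabric ribbon material for Niagara ribbons in the DimeBlanket skydiving contrail.
# One fabric texture is tiled and panned by a vector parameter, using only R for both tiling axes and B for both speed axes.
# The same sample's RGB is tinted by particle color and also feeds DeriveNormalZ, whose normal is multiplied by TwoSidedSign to flip on back faces.
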Material /BRCosmetics/Athena/Effects/Particles/Skydiving/NS_Contrail_DimeBlanket/M_Contrail_DimeBlanket_FabricRibbon
graph {
 properties {bAutomaticallySetUsageInEditor=false, bUsedWithNiagaraRibbons=true, twoSided=true, usedShadingModels=Default Lit}
 nodes [
  n1 = TextureCoordinate()
  n2 = VectorParameter() {parameterName=Main Tex UV/Speed, defaultValue=<rgba 1 1 0 0>}
  n3 = AppendVector(A=@n2.R, B=@n2.R)
  n4 = TextureSample(UVs=Panner(Coordinate=Multiply(A=@n1, B=@n3), Speed=AppendVector(A=@n2.B, B=@n2.B))) {texture=&/T_Contrail_DimeBlanket_FabricRibbon}
  Multiply(A=ParticleColor().RGB, B=@n4.RGB)
  Multiply(
   A=MaterialFunctionCall!(XY vector=@n4.RGB, Z sign=Constant() {r=1}, Default Normalization vector) {materialFunction=/Engine/Functions/Engine_MaterialFunctions02/Math/DeriveNormalZ_Function}.Result
   B=TwoSidedSign()
  )
 ]
}
