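# Translucent unlit Niagara sprite material for candy-heart particles, filed under the DimeBlanket skydiving contrail effect.
# A SubUV function steps through a grid of heart mask frames using time times a preview speed parameter, rather than particle-driven frame selection.
# The mask alpha, particle alpha and near-camera fade multiply into a depth fade for soft opacity.
# Particle color supplies emissive through a shared FX emissive adjustment function.
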Material /BRCosmetics/Athena/Effects/Particles/Skydiving/NS_Contrail_DimeBlanket/M_SubUV_Hearts
graph {
 properties {
  bAutomaticallySetUsageInEditor=false
  bEnableResponsiveAA=true
  bGenerateSphericalParticleNormals=true
  bUsedWithNiagaraSprites=true
  blendMode=BLEND_Translucent
  pixelDepthOffsetMode=PDOM_Legacy
  shadingModel=MSM_Unlit
  usedShadingModels=Unlit
 }
 nodes [
  n1 = ParticleColor()
  n2 = MaterialFunctionCall!(
   Texture=TextureObjectParameter() {parameterName=Alpha_0, texture=/Game/Gadgets/Assets/VinderTech_GliderChute/Glider_TeddyBear/Effects/Textures/T_Candyheart_Mask}
   UVs
   SubImages=AppendVector(
    A=ScalarParameter() {defaultValue=4, parameterName=Confetti0_U_Repeats}
    B=ScalarParameter() {defaultValue=4, parameterName=Confetti0_V_Repeats}
   )
   Frame=Multiply(A=Time(), B=ScalarParameter() {defaultValue=15, parameterName=Preview SubUV Speed})
  ) {materialFunction=/Engine/Functions/Engine_MaterialFunctions03/Texturing/SubUV_Function}
  n3 = ScalarParameter() {defaultValue=64, parameterName=NearFadeDistance}
  DepthFade(
   Opacity=Multiply(
    A=Multiply(A=@n1.A, B=@n2.Alpha)
    B=MaterialFunctionCall(Distance=@n3) {materialFunction=&Near_Camera_Fade}.Result
   )
   FadeDistance=ScalarParameter() {defaultValue=128, parameterName=FadeDistance}
  )
  MaterialFunctionCall(Emissive=@n1.RGB) {materialFunction=&MF_FX_EmissiveAdjustments}
 ]
}
