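# Niagara skydiving contrail cosmetic (DimeBlanket) with spiraling fabric ribbons, pink smoke, magical dust, hearts and kiss sparkles.
# Fabric, dust and smoke place particles along a SmokeTrail emitter's ribbons through a particle-read FollowRibbon module, with fabric adding sine/cosine spiral offsets.
# Hearts and kisses spawn at skeletal bones or sockets, drifting with curl noise and inherited owner motion.
# Dive alphas, time-of-day scaling, camera fade and front-end preview overrides tune the look.
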NiagaraSystem /BRCosmetics/Athena/Effects/Particles/Skydiving/NS_Contrail_DimeBlanket/NS_Contrail_DimeBlanket
dependencies {
 usedDataInterfaces [
  &NiagaraDataInterfaceCurve
  &NiagaraDataInterfaceParticleRead
  &NiagaraDataInterfaceVectorCurve
  &NiagaraDataInterfaceColorCurve
  &NiagaraDataInterfaceArrayFloat
  &NiagaraDataInterfaceArrayFloat3
  &NiagaraDataInterfaceVector2DCurve
  &NiagaraDataInterfaceCamera
  &NiagaraDataInterfaceSkeletalMesh
 ]
 usedDynamicInputs [&LerpLinearColors, /Niagara/DynamicInputs/Lerp/SmoothLerpOverTime_Float]
 usedModules [
  &self:SystemSpawnScript.SystemScriptSource.SystemScriptGraph.NiagaraNodeAssignment_4.SetVariables_#1
  &self:SystemSpawnScript.SystemScriptSource.SystemScriptGraph.NiagaraNodeAssignment_6.SetVariables_#2
  &SystemState
  /Game/Effects/Niagara/Modules/Skydiving/FN_DivingAttributes/v5/FN_DivingAttributes
  &self:SystemSpawnScript.SystemScriptSource.SystemScriptGraph.NiagaraNodeAssignment_2.SetVariables_#3
  &self:SystemSpawnScript.SystemScriptSource.SystemScriptGraph.NiagaraNodeAssignment_8.SetVariables_#4
 ]
}
emitters {
 FabricRibbon {
  emitterSpawnScript [/SetVariables(Emitter.ParticlePerRibbon=20, Emitter.RibbonNumber=2) {isSetParameters=true}]
  emitterUpdateScript [
   &EmitterState(
    Inactive Response=::NewEnumerator1
    Life Cycle Mode=::NewEnumerator1
    Loop Behavior=::NewEnumerator1
    Loop Duration Mode=::NewEnumerator1
    UseLoopDelay=false
   )
   &SpawnBurst_Instantaneous(
    Spawn Count=&Multiply_Int(A=$Emitter.ParticlePerRibbon, B=$Emitter.RibbonNumber)
    Use Spawn Probability=false
    UseLoopCountLimit=false
   )
  ]
  particleSpawnScript [
   &InitializeParticle(Ribbon Width=13, Ribbon Width Mode=::NewEnumerator1, Sprite Size Mode=::NewEnumerator3, UsePositionOffset=false, Write Material Random=true)
   &PartitionParticles(
    ID to Partition:NiagaraInt32=&ReturnExecIndex()
    Normalized Index In Partition Includes One=true
    Number Of Partitions=$Emitter.RibbonNumber
    Particles Per Partition:NiagaraInt32=$Emitter.ParticlePerRibbon
    Partition Mode:ENiagara_PartitionMode=::NewEnumerator1
   )
   /SetVariables(Particles.NormalizedRibbonPos:NiagaraFloat=&OneMinusFloat(Float=$Particles.PartitionParticles.Normalized Index In Partition)) {isSetParameters=true}
   /SetVariables(
    Particles.SpiralFreq=0.06
    Particles.SpiralStrength:NiagaraFloat=&Multiply_Float(
     A=20
     B=&FloatFromCurve(
      FloatCurve=DataInterface() {propertyValues {Curve {keys [
          {arriveTangent=3.6599, interpMode=RCIM_Cubic, leaveTangent=3.6599, tangentMode=RCTM_User, time=0, value=-0.025}
          {interpMode=RCIM_Cubic, tangentMode=RCTM_User, time=0.5, value=1}
          {arriveTangent=-0.1348, interpMode=RCIM_Cubic, leaveTangent=-0.1348, tangentMode=RCTM_User, time=1, value=0}
         ]}}}
      CurveIndex=$Particles.NormalizedRibbonPos
     )
    )
   ) {isSetParameters=true}
   /SetVariables(Particles.RibbonU0Override=&OneMinusFloat(Float=$Particles.NormalizedRibbonPos)) {isSetParameters=true}
   &CameraOffset()
  ]
  particleUpdateScript [
   &ParticleState(Kill Particles When Lifetime Has Elapsed=false)
   /SetVariables(Particles.SpiralStrength:NiagaraFloat=&Multiply_Float(
     A=1.5
     B=&FloatFromCurve(
      FloatCurve=DataInterface() {propertyValues {Curve {keys [
          {arriveTangent=24.2953, interpMode=RCIM_Cubic, time=0.05, value=0.025}
          {arriveTangent=29.3611, interpMode=RCIM_Cubic, leaveTangent=11.1315, tangentMode=RCTM_Break, time=0.12, value=1}
          {arriveTangent=0.0457, interpMode=RCIM_Cubic, leaveTangent=-0.8624, tangentMode=RCTM_Break, time=0.2, value=1.5}
          {arriveTangent=-2.7337, interpMode=RCIM_Cubic, leaveTangent=-2.7337, tangentMode=RCTM_User, time=1, value=0.1}
         ]}}}
      CurveIndex=$Particles.NormalizedRibbonPos
     )
    )) {isSetParameters=true}
   /SetVariables(Particles.SpiralTime:NiagaraFloat=&Add_Float(
     A=&Add_Float(A=&Multiply_Float(A=$System.Age, B=0.08), B=$Particles.NormalizedRibbonPos)
     B=&Multiply_FloatByInt(
      Float=&Divide_Float(A=$Particles.SpiralFreq, B=&MakeFloatFromInt(Int=$Emitter.RibbonNumber))
      Integer=$Particles.PartitionParticles.Partition
     )
    )) {isSetParameters=true}
   /SetVariables(Particles.SpiralOffset:Vector3f=&MakeVector(
     X=&Sine(
      Normalized Angle=$Particles.SpiralTime
      Period=$Particles.SpiralFreq
      Scale=&Multiply_Float(A=$Particles.SpiralStrength, B=10)
      Bias=0
     )
     Y=&Cosine(
      Normalized Angle=$Particles.SpiralTime
      Period=$Particles.SpiralFreq
      Scale=&Multiply_Float(A=$Particles.SpiralStrength, B=10)
      Bias=0
     )
     Z=0
    )) {isSetParameters=true}
   &self:FollowRibbon(
    Alpha:NiagaraFloat=&FloatFromCurve(
     FloatCurve=DataInterface() {propertyValues {Curve {keys [{tangentMode=RCTM_User, time=0.05, value=0}, {tangentMode=RCTM_User, time=1, value=1}]}}}
     CurveIndex=$Particles.Initial.NormalizedRibbonPos
    )
    ParticlePerRibbon:NiagaraInt32=$Emitter.ParticlePerRibbon
    RibbonNumber:NiagaraInt32=$Particles.PartitionParticles.Partition
    RibbonReader=DataInterface() {propertyValues {EmitterBinding {bindingMode=Other, emitterName=SmokeTrail}}}
   ) {module=FollowRibbon}
   /SetVariables(Particles.Position=$Output.FollowRibbon.RibbonPos) {isSetParameters=true}
   /SetVariables(Particles.Velocity=$Output.FollowRibbon.DirectionVector) {isSetParameters=true}
   /SetVariables(Particles.Position=&AddVectorToPosition(
     Position=&AddVectorToPosition(
      Position=$Particles.Position
      Vector=&VectorFromCurve(
       VectorCurve=DataInterface() {propertyValues {XCurve {keys [<at 0 1>,<at 1 0>]}, YCurve {keys [<at 0 1>,<at 1 0>]}, ZCurve {keys [<at 0.02 0>,<at 0.1 1>,<at 0.3 0>]}}}
       CurveIndex=$Particles.NormalizedRibbonPos
       Scale Curve=<0 0 4>
      )
     )
     Vector=$Particles.SpiralOffset
    )) {isSetParameters=true}
   /SetVariables(Particles.RibbonFacing=&NormalizeVector(
     Vector To Normalize=&AddVectorToPosition(
      Position=Linked() {linkedVariable {name=Particles.SpiralOffset, type=Vector3f}}
      Vector=&VectorFromCurve(
       VectorCurve=DataInterface() {propertyValues {XCurve {keys [<at 0 1>,<at 1 0>]}, YCurve {keys [<at 0 1>,<at 1 0>]}, ZCurve {keys [<at 0.02 0>,<at 0.1 1>,<at 0.3 0>]}}}
       CurveIndex=$Particles.NormalizedRibbonPos
       Scale Curve=<0 0 -1>
      )
     )
     Safe Normalize=true
     Safe Fallback Threshold=0
     Safe Fallback Vector=<0 0 0.001>
    )) {isSetParameters=true}
   &DynamicMaterialParameters(
    Index 0 Param 1=-0.255
    Index 0 Param 2=-0.1
    Index 0 Param 3=1
    Module.Index0Param0Write=true
    Module.Index0Param1Write=false
    Module.Index0Param2Write=false
    Module.Index0Param3Write=false
   ) {module=DynamicMaterialParameters001}
   &ScaleColor(
    Curve Index=$Particles.NormalizedRibbonPos
    Linear Color Curve=DataInterface() {propertyValues {
      AlphaCurve {keys [<at 0 0>,<at 0.0274 1 RCIM_Cubic>,<at 1 0>]}
      BlueCurve {keys [<at 0 1>]}
      GreenCurve {keys [<at 0 1>]}
      RedCurve {keys [<at 0 1>]}
     }}
    Scale Mode=::NewEnumerator2
   ) {module=ScaleColor001}
   &ScaleColor(
    Scale Alpha=&FloatFromCurve(
     FloatCurve=DataInterface() {propertyValues {Curve {keys [
         {arriveTangent=8.2127, interpMode=RCIM_Cubic, leaveTangent=8.2127, tangentMode=RCTM_User, time=0.12, value=0.05}
         {arriveTangent=0.4596, arriveTangentWeight=0.232, leaveTangentWeight=2.2538, time=0.22, value=1}
         {arriveTangent=0.0486, interpMode=RCIM_Cubic, leaveTangent=0.0486, tangentMode=RCTM_User, time=0.75, value=1}
         {arriveTangent=-4.4982, interpMode=RCIM_Cubic, leaveTangent=-4.4982, tangentMode=RCTM_User, time=1, value=0}
        ]}}}
     CurveIndex=$Particles.NormalizedRibbonPos
     Scale Curve=&OneMinusFloat(Float=$System.Timeline.NormalizedTime)
    )
    ScaleA=true
    ScaleRGB=true
   )
   &ScaleRibbonWidth(Ribbon Width Scale=&FloatFromCurve(
     FloatCurve=DataInterface() {propertyValues {Curve {keys [
         {arriveTangent=15.9799, interpMode=RCIM_Cubic, leaveTangent=15.9799, tangentMode=RCTM_User, time=0.02, value=0}
         {arriveTangent=0.0354, tangentMode=RCTM_User, time=0.1, value=1}
         {arriveTangent=-0.3163, interpMode=RCIM_Cubic, leaveTangent=-0.3163, tangentMode=RCTM_User, time=0.5, value=1}
         {arriveTangent=-5.4829, interpMode=RCIM_Cubic, leaveTangent=-5.4829, tangentMode=RCTM_User, time=0.8, value=0}
        ]}}}
     CurveIndex=$Particles.NormalizedRibbonPos
    )) {module=ScaleRibbonWidth001}
  ]
  rendererClasses [&NiagaraRibbonRendererProperties]
  renderers [{rendererClass=&NiagaraRibbonRendererProperties}]
 }
 Hearts {
  emitterUpdateScript [
   &EmitterState(Life Cycle Mode=::NewEnumerator1, Loop Behavior=::NewEnumerator1, Loop Duration Mode=::NewEnumerator1, UseLoopDelay=false)
   &SpawnRate(Spawn Probability=0.33, SpawnRate=20, Use Spawn Probability=false)
  ]
  particleSpawnScript [
   &FN_SetSpawnGroup()
   &InitializeParticle(
    Lifetime Max=&SelectFloatFromArray(Direct Array Index=$Particles.SpawnGroup, Float Selection Array=DataInterface() {propertyValues {FloatData [1.5,1.5,0.75,0.75]}})
    Lifetime Min:NiagaraFloat=&SelectFloatFromArray(Direct Array Index=$Particles.SpawnGroup, Float Selection Array=DataInterface() {propertyValues {FloatData [1,1,0.5,0.5]}})
    Lifetime Mode=::NewEnumerator1
    Sprite Rotation Mode=::NewEnumerator2
    Sprite Size Mode=::NewEnumerator1
    Uniform Sprite Size Max=&SelectFloatFromArray(Direct Array Index=$Particles.SpawnGroup, Float Selection Array=DataInterface() {propertyValues {FloatData [8,8,4,4]}})
    Uniform Sprite Size Min=&SelectFloatFromArray(Direct Array Index=$Particles.SpawnGroup, Float Selection Array=DataInterface() {propertyValues {FloatData [4,4,2,2]}})
    UsePositionOffset=false
    Write Material Random=true
   )
   &SkeletalMeshLocation(
    Bone / Socket Index=$Particles.SpawnGroup
    Bone Sampling Mode=::NewEnumerator4
    Interpolate Mesh Sampling=false
    MaskSpawnGroup=false
    Sampled Position Offset=<0 0 -20>
    Sampled Position Offset Coordinate Space:ENiagaraSkelMeshTransforms=::NewEnumerator2
    Skeletal Mesh=$System.Skeletal Mesh
   ) {module=SkeletalMeshLocation001}
   &ShapeLocation(
    MaskSpawnGroup=false
    Offset=<0 0 10>
    Offset Mode=::NewEnumerator0
    Override Seed=false
    Sphere Radius=&SelectFloatFromArray(
     Direct Array Index=&SpawnGroup()
     Float Selection Array=DataInterface() {propertyValues {FloatData [30,30,10,10]}}
    )
    UseOwnerScale=true
   )
   &AddVelocity(
    Cone Angle=1
    Cone Axis=&FN_FrontEndVector3Modifier(Gameplay=<0 0 1>, Front End=&Lerp_Vector(A=<0 0 1>, B=<0 -1 0.75>, Alpha=$System.Dive))
    Module.UseConeFalloff=false
    Module.UseVelDistribution=false
    Override Seed=false
    Velocity Mode=::NewEnumerator2
    Velocity Speed=&Lerp_Float(A=200, B=400, Alpha=$System.ContrailDiveAlpha)
   )
  ]
  particleUpdateScript [
   &ParticleState()
   &SpriteRotationRate(Rotation Rate=&RandomRangeFloat(Minimum=-35, Maximum=35))
   &ScaleSpriteSize(
    Uniform Curve Scale=&Multiply_Float(
     A=0.77
     B=/Niagara/DynamicInputs/Camera/CameraFade(
      Camera Query:NiagaraDataInterfaceCamera=DataInterface()
      Fade Offset:NiagaraFloat=0
      Fade Distance:NiagaraFloat=150
      Use Owner Position:NiagaraBool=true
     )
    )
    Uniform Curve Sprite Scale=DataInterface() {propertyValues {Curve {keys [
        {interpMode=RCIM_Cubic, tangentMode=RCTM_User, time=0, value=0}
        {arriveTangent=1.2783, interpMode=RCIM_Cubic, leaveTangent=1.2783, tangentMode=RCTM_User, time=0.4552, value=1.8214}
        {arriveTangent=-0.3706, interpMode=RCIM_Cubic, leaveTangent=-0.3706, tangentMode=RCTM_User, time=0.5846, value=1.7396}
        {arriveTangent=-2.6828, interpMode=RCIM_Cubic, leaveTangent=-2.6828, tangentMode=RCTM_User, time=0.7207, value=2.0221}
        <at 1 0.25>
       ]}}}
   )
   &ScaleColor(
    Scale Alpha=&FloatFromCurve(FloatCurve=DataInterface() {propertyValues {Curve {keys [<at 0.6 1 RCIM_Cubic>,<at 1 1>]}}})
    Scale RGB=&MakeVectorFromLinearColorRGB(Color=&ColorFromCurve(ColorCurve=DataInterface() {propertyValues {
        AlphaCurve {keys [<at 0 1 RCIM_Cubic>]}
        BlueCurve {keys [<at 0.2232 0 RCIM_Cubic>,<at 1 0.302 RCIM_Cubic>]}
        GreenCurve {keys [<at 0.2232 0 RCIM_Cubic>,<at 1 0.1686 RCIM_Cubic>]}
        RedCurve {keys [<at 0.2232 1 RCIM_Cubic>,<at 1 0.9294 RCIM_Cubic>]}
       }}))
    ScaleA=true
    ScaleRGB=true
   ) {module=ScaleColor002}
   &CurlNoiseForce(
    Bias Noise=false
    Noise Frequency=&RandomRangeFloat(
     Minimum=&SelectFloatFromArray(Direct Array Index=$Particles.SpawnGroup, Float Selection Array=DataInterface() {propertyValues {FloatData [100,100,50,50]}})
     Maximum=&SelectFloatFromArray(Direct Array Index=$Particles.SpawnGroup, Float Selection Array=DataInterface() {propertyValues {FloatData [200,200,100,100]}})
    )
    Noise Strength=&FloatFromCurve(
     FloatCurve=DataInterface() {propertyValues {Curve {keys [
         {arriveTangent=4.1785, interpMode=RCIM_Cubic, leaveTangent=4.1785, tangentMode=RCTM_User, time=0, value=0}
         {arriveTangent=1.1179, interpMode=RCIM_Cubic, leaveTangent=1.1179, tangentMode=RCTM_User, time=0.2042, value=0.6228}
         {arriveTangent=0.0224, interpMode=RCIM_Cubic, leaveTangent=0.0224, tangentMode=RCTM_User, time=1, value=1}
        ]}}}
     Scale Curve=&RandomRangeFloat(
      Minimum=&SelectFloatFromArray(Direct Array Index=$Particles.SpawnGroup, Float Selection Array=DataInterface() {propertyValues {FloatData [1000,1000,500,500]}})
      Maximum=&SelectFloatFromArray(Direct Array Index=$Particles.SpawnGroup, Float Selection Array=DataInterface() {propertyValues {FloatData [1000,1000,500,500]}})
     )
    )
    Pan Noise=false
   )
   &InheritSourceMovement(
    Applied Position Scale=&Multiply_Vector(
     A=&VectorFromFloat(Value=&Multiply_Float(
       A=&Lerp_Float(A=0.25, B=-1, Alpha=$System.Dive)
       B=&FloatFromCurve(FloatCurve=DataInterface() {propertyValues {Curve {keys [{tangentMode=RCTM_User, time=0, value=0}, {tangentMode=RCTM_User, time=1, value=1}]}}})
      ))
     B=<0.5 1 0>
    )
    ApplyForce=false
    ApplyPosition=true
    ApplyPositionFalloff=false
    Source Velocity:Vector3f=&TransformVector(Vector=$Engine.Owner.Velocity, Source Space=::World, Destination Space=::Local)
   )
   &SolveForcesAndVelocity(Clamp Velocity=false, Limit Acceleration=false)
   &ScaleColor(
    Scale RGB=&VectorFromFloat(Value=&FNTimeOfDayScalar(Night Scale=0.9, Morning Scale=1, Evening Scale=1))
    ScaleA=true
    ScaleRGB=true
   )
   &FN_Sprite_Attributes(
    Create Color Attribute=true
    Create Position Attribute=true
    Create SpriteSize Attribute=true
    SpriteSize=&Multiply_Vector2D_ByFloat(Float=0.75, Vector2D=$Particles.SpriteSize)
    Write to Module Namespace=true
   )
  ]
  rendererClasses [&NiagaraSpriteRendererProperties]
  renderers [
   {rendererClass=&NiagaraSpriteRendererProperties}
   {rendererClass=&NiagaraSpriteRendererProperties, rendererIndex=1}
  ]
 }
 Kisses_Sparkles {
  emitterSpawnScript [/SetVariables(Emitter.RandomStickerIndex:NiagaraInt32=&RandomRangeInt(Minimum=0, Maximum=1)) {isSetParameters=true}]
  emitterUpdateScript [
   &EmitterState(Life Cycle Mode=::NewEnumerator1, Loop Behavior=::NewEnumerator1, Loop Duration Mode=::NewEnumerator1, UseLoopDelay=false)
   &SpawnRate(Spawn Probability=0.33, SpawnRate=20, Use Spawn Probability=false)
  ]
  particleSpawnScript [
   &FN_SetSpawnGroup(Number of Spawn Groups=2)
   /SetVariables(Particles.CustomIndex:NiagaraInt32=&ModuloInt(
     A=&Add_Integer(A=$Emitter.RandomStickerIndex, B=&ReturnParticleIDAttributes())
     B=4
    )) {isSetParameters=true}
   /SetVariables(Particles.VisibilityTag=&MakeCustomIntFromBool(
     Bool=&SetBoolByIntComparison(A=$Particles.CustomIndex, B=1, Comparison Type=::NewEnumerator0)
     True INT=1
     False INT=0
    )) {isSetParameters=true}
   &InitializeParticle(
    Color Minimum=<rgba 0.9294 0.1686 0.302 1>
    Color Mode=::NewEnumerator2
    Lifetime Max=1.25
    Lifetime Mode=::NewEnumerator1
    Sprite Size Mode=::NewEnumerator3
    Uniform Sprite Size=&RandomRangeFloat(Minimum=4, Maximum=8)
    UsePositionOffset=false
    Write Material Random=true
   )
   &SkeletalMeshLocation(
    Bone / Socket Index=$Particles.SpawnGroup
    Bone Sampling Mode=::NewEnumerator4
    Interpolate Mesh Sampling=false
    MaskSpawnGroup=false
    Sampled Position Offset=<0 0 -20>
    Sampled Position Offset Coordinate Space:ENiagaraSkelMeshTransforms=::NewEnumerator2
    Skeletal Mesh=$System.Skeletal Mesh
   ) {module=SkeletalMeshLocation001}
   &ShapeLocation(MaskSpawnGroup=false, Offset=<0 0 10>, Offset Mode=::NewEnumerator0, Override Seed=false, Sphere Radius=30, UseOwnerScale=true)
   &AddVelocity(
    Cone Angle=1
    Cone Axis=&FN_FrontEndVector3Modifier(Gameplay=<0 0 1>, Front End=&Lerp_Vector(A=<0 0 1>, B=<0 -1 0.75>, Alpha=$System.Dive))
    Module.UseConeFalloff=false
    Module.UseVelDistribution=false
    Override Seed=false
    Velocity Mode=::NewEnumerator2
    Velocity Speed=&Lerp_Float(A=200, B=400, Alpha=$System.ContrailDiveAlpha)
   )
   /SetVariables(Particles.GlowCameraOffset=-0.1) {isSetParameters=true}
  ]
  particleUpdateScript [
   &ParticleState()
   &SpriteRotationRate(Rotation Rate=&RandomRangeFloat(Minimum=-35, Maximum=35))
   &ScaleSpriteSize(
    Non-Uniform Curve Sprite Scale=DataInterface() {propertyValues {
      XCurve {keys [
        {interpMode=RCIM_Cubic, tangentMode=RCTM_User, time=0.05, value=0.9}
        {arriveTangent=0.4905, interpMode=RCIM_Cubic, leaveTangent=0.4905, tangentMode=RCTM_User, time=0.35, value=0.65}
        {arriveTangent=1.7384, interpMode=RCIM_Cubic, leaveTangent=1.7384, tangentMode=RCTM_User, time=0.65, value=1.05}
        {interpMode=RCIM_Cubic, leaveTangent=0.1643, time=0.85, value=0.75}
       ]}
      YCurve {keys [
        {arriveTangent=-0.6287, interpMode=RCIM_Cubic, leaveTangent=-0.6287, tangentMode=RCTM_User, time=0, value=1.025}
        {arriveTangent=-0.3763, interpMode=RCIM_Cubic, leaveTangent=-0.3763, tangentMode=RCTM_User, time=0.25, value=1}
        {arriveTangent=-0.5, interpMode=RCIM_Cubic, leaveTangent=-0.5, time=0.45, value=1.15}
        {arriveTangent=-1.625, interpMode=RCIM_Cubic, leaveTangent=-1.625, time=0.65, value=0.8}
        <at 0.85 0.5 RCIM_Cubic>
       ]}
     }}
    Scale Sprite Size Mode=::NewEnumerator3
   ) {module=ScaleSpriteSize001}
   &ScaleSpriteSize(
    Uniform Curve Scale=&Multiply_Float(
     A=0.77
     B=&CameraFade(
      Camera Query:NiagaraDataInterfaceCamera=DataInterface()
      Fade Offset:NiagaraFloat=0
      Fade Distance:NiagaraFloat=150
      Use Owner Position:NiagaraBool=true
     )
    )
    Uniform Curve Sprite Scale=DataInterface() {propertyValues {Curve {keys [
        {interpMode=RCIM_Cubic, tangentMode=RCTM_User, time=0, value=0}
        {arriveTangent=-0.371, interpMode=RCIM_Cubic, leaveTangent=-0.371, tangentMode=RCTM_User, time=0.2, value=1.875}
        {arriveTangent=-0.3706, interpMode=RCIM_Cubic, leaveTangent=-0.3706, tangentMode=RCTM_User, time=0.5846, value=1.7396}
        {arriveTangent=-2.6828, interpMode=RCIM_Cubic, leaveTangent=-2.6828, tangentMode=RCTM_User, time=0.7207, value=2.0221}
        <at 1 0.25>
       ]}}}
   )
   &DynamicMaterialParameters(
    Index 0 Param 1=&FloatFromCurve(FloatCurve=DataInterface() {propertyValues {Curve {keys [
         {arriveTangent=-0.0454, interpMode=RCIM_Cubic, time=0.41, value=0}
         {interpMode=RCIM_Cubic, tangentMode=RCTM_Break, time=0.5, value=0.27}
         {arriveTangent=-0.1021, interpMode=RCIM_Cubic, leaveTangent=-0.1021, time=0.6, value=0.2599}
         <at 0.8 0.2394 RCIM_Cubic>
        ]}}})
    Index 0 Param 2=&FloatFromCurve(FloatCurve=DataInterface() {propertyValues {Curve {keys [
         {arriveTangentWeight=0.05, interpMode=RCIM_Cubic, leaveTangentWeight=0.2833, tangentMode=RCTM_User, time=0, value=0.9}
         {arriveTangentWeight=0.1167, interpMode=RCIM_Cubic, leaveTangentWeight=0.1667, tangentMode=RCTM_Break, time=0.3297, value=0.9}
         {arriveTangentWeight=0.1667, interpMode=RCIM_Cubic, tangentMode=RCTM_User, time=0.4693, value=-0.9025}
         <at 0.8 -0.1217 RCIM_Cubic>
        ]}}})
    Index 0 Param 3=&FloatFromCurve(FloatCurve=DataInterface() {propertyValues {Curve {keys [
         {arriveTangentWeight=0.1524, interpMode=RCIM_Cubic, leaveTangentWeight=0.3556, tangentMode=RCTM_User, time=0.42, value=0.0974}
         {arriveTangentWeight=0.2333, interpMode=RCIM_Cubic, tangentMode=RCTM_User, time=0.47, value=1}
        ]}}})
    Index 0 Param 4=&FloatFromCurve(FloatCurve=DataInterface() {propertyValues {Curve {keys [<at 0.36 0.14 RCIM_Cubic>,<at 0.4437 0>]}}})
    Index 1 Param 1=&FloatFromCurve(FloatCurve=DataInterface() {propertyValues {Curve {keys [<at 0.3719 1 RCIM_Cubic>,<at 0.4809 0>]}}})
    Module.Index0Param0Write=true
    Module.Index0Param1Write=true
    Module.Index0Param2Write=true
    Module.Index0Param3Write=true
    Module.Index1Param0Write=true
    Module.Index1Param1Write=true
    Module.Index1Param2Write=true
    Module.Index1Param3Write=true
    Write Parameter Index 1=true
   )
   &CurlNoiseForce(
    Bias Noise=false
    Noise Frequency=&RandomRangeFloat(Minimum=100, Maximum=200)
    Noise Strength=&FloatFromCurve(
     FloatCurve=DataInterface() {propertyValues {Curve {keys [{tangentMode=RCTM_User, time=0, value=0}, {tangentMode=RCTM_User, time=1, value=1}]}}}
     Scale Curve=&RandomRangeFloat(Minimum=1000, Maximum=2000)
    )
    Pan Noise=false
   )
   &InheritSourceMovement(
    Applied Position Scale=&Multiply_Vector(
     A=&VectorFromFloat(Value=&Multiply_Float(
       A=&Lerp_Float(A=0.25, B=-1, Alpha=$System.Dive)
       B=&FloatFromCurve(FloatCurve=DataInterface() {propertyValues {Curve {keys [{tangentMode=RCTM_User, time=0, value=0}, {tangentMode=RCTM_User, time=1, value=1}]}}})
      ))
     B=<0.5 1 0>
    )
    ApplyForce=false
    ApplyPosition=true
    ApplyPositionFalloff=false
    Source Velocity:Vector3f=&TransformVector(Vector=$Engine.Owner.Velocity, Source Space=::World, Destination Space=::Local)
   )
   &ScaleColor(
    Scale Alpha=&FloatFromCurve(FloatCurve=DataInterface() {propertyValues {Curve {keys [<at 0.6 1 RCIM_Cubic>,<at 1 1>]}}})
    Scale RGB=&VectorFromFloat(Value=&FloatFromCurve(FloatCurve=DataInterface() {propertyValues {Curve {keys [<at 0 2.45 RCIM_Cubic>, {arriveTangent=-1.55, interpMode=RCIM_Cubic, leaveTangent=-1.55, time=0.3, value=1.25}, <at 1 0.9 RCIM_Cubic>]}}}))
    ScaleA=true
    ScaleRGB=true
   ) {module=ScaleColor001}
   &SolveForcesAndVelocity(Clamp Velocity=false, Limit Acceleration=false)
   /SetVariables(
    Particles.DarkGlow:LinearColor=&MakeCustomLinearColorFromBool(
     Bool=$User.bIsFrontEndPreview
     True Color=&ColorFromCurve(ColorCurve=DataInterface() {propertyValues {
        AlphaCurve {keys [<at 0 0 RCIM_Cubic>, {arriveTangent=0.0275, interpMode=RCIM_Cubic, leaveTangent=0.0275, time=0.0509, value=0.02}, <at 0.7275 0.02>, <at 1 0>]}
        BlueCurve {keys [<at 0 0 RCIM_Cubic>]}
        GreenCurve {keys [<at 0 0 RCIM_Cubic>]}
        RedCurve {keys [<at 0 2 RCIM_Cubic>]}
       }})
     False Color=&MultiplyLinearColor(
      Linear Color=&ColorFromCurve(ColorCurve=DataInterface() {propertyValues {
         AlphaCurve {keys [<at 0 0 RCIM_Cubic>, {arriveTangent=0.4124, interpMode=RCIM_Cubic, leaveTangent=0.4124, time=0.0509, value=0.3}, <at 0.7275 0.3>, <at 1 0>]}
         BlueCurve {keys [<at 0 0 RCIM_Cubic>]}
         GreenCurve {keys [<at 0 0 RCIM_Cubic>]}
         RedCurve {keys [<at 0 1 RCIM_Cubic>]}
        }})
      Scale Factor=&MakeVector4_FromFloat(Float=&FNTimeOfDayScalar(Night Scale=0.7, Morning Scale=1, Evening Scale=0.8))
     )
    )
    Particles.GlowSize=&Multiply_Vector2D(A=$Particles.SpriteSize, B=<2 2>)
   ) {isSetParameters=true}
   &ScaleColor(
    Scale RGB=&VectorFromFloat(Value=&FNTimeOfDayScalar(Night Scale=0.9, Morning Scale=1, Evening Scale=1))
    ScaleA=true
    ScaleRGB=true
   )
   /SetVariables(
    Particles.SparkleBackColor:LinearColor=&MultiplyLinearColor(
     Linear Color=&MultiplyLinearColor(
      Linear Color=<rgba 1 1 1 1>
      Scale Factor=&MakeFloat4FromVectorAndFloat(
       "Vector (RGB)"=&VectorFromFloat(Value=&FNTimeOfDayScalar(Night Scale=0.9, Morning Scale=1, Evening Scale=1))
       "Float (Alpha)"=1
      )
     )
     Scale Factor=&MakeVector4_FromLinearColor(LinearColor=&ColorFromCurve(ColorCurve=DataInterface() {propertyValues {
         AlphaCurve {keys [<at 0 1 RCIM_Cubic>]}
         BlueCurve {keys [<at 0.2232 0 RCIM_Cubic>,<at 1 0.302 RCIM_Cubic>]}
         GreenCurve {keys [<at 0.2232 0 RCIM_Cubic>,<at 1 0.1686 RCIM_Cubic>]}
         RedCurve {keys [<at 0.2232 1 RCIM_Cubic>,<at 1 0.9294 RCIM_Cubic>]}
        }}))
    )
    Particles.SparkleSize:Vector2f=&Multiply_Vector2D(
     A=&RandomRangeVector2D(Minimum=<15 15>, Maximum=<20 25>)
     B=&Multiply_Vector2D_ByFloat(
      Float=&FloatFromCurve(
       FloatCurve=DataInterface() {propertyValues {Curve {keys [
           {interpMode=RCIM_Cubic, tangentMode=RCTM_User, time=0, value=0}
           {arriveTangent=0.1421, interpMode=RCIM_Cubic, leaveTangent=0.1421, tangentMode=RCTM_User, time=0.2, value=1.875}
           {arriveTangent=-0.3706, interpMode=RCIM_Cubic, leaveTangent=-0.3706, tangentMode=RCTM_User, time=0.5846, value=1.7396}
           {arriveTangent=-2.6828, interpMode=RCIM_Cubic, leaveTangent=-2.6828, tangentMode=RCTM_User, time=0.7207, value=2.0221}
           <at 1 0.25>
          ]}}}
       Scale Curve=&Multiply_Float(
        A=0.77
        B=&CameraFade(
         Camera Query:NiagaraDataInterfaceCamera=DataInterface()
         Fade Offset:NiagaraFloat=0
         Fade Distance:NiagaraFloat=150
         Use Owner Position:NiagaraBool=true
        )
       )
      )
      Vector2D=&Vector2DFromCurve(
       Vector2Curve=DataInterface() {propertyValues {
         XCurve {keys [
           {interpMode=RCIM_Cubic, tangentMode=RCTM_User, time=0.1, value=0.9}
           {arriveTangent=0.4905, interpMode=RCIM_Cubic, leaveTangent=0.4905, tangentMode=RCTM_User, time=0.4, value=0.65}
           {arriveTangent=1.7384, interpMode=RCIM_Cubic, leaveTangent=1.7384, tangentMode=RCTM_User, time=0.7, value=1.05}
           {interpMode=RCIM_Cubic, leaveTangent=0.1643, time=0.9, value=0.75}
          ]}
         YCurve {keys [
           {arriveTangent=-0.6287, interpMode=RCIM_Cubic, leaveTangent=-0.6287, tangentMode=RCTM_User, time=0, value=1.025}
           {arriveTangent=-0.3763, interpMode=RCIM_Cubic, leaveTangent=-0.3763, tangentMode=RCTM_User, time=0.3, value=1}
           {arriveTangent=-0.5, interpMode=RCIM_Cubic, leaveTangent=-0.5, time=0.5, value=1.15}
           {arriveTangent=-1.625, interpMode=RCIM_Cubic, leaveTangent=-1.625, time=0.7, value=0.8}
           <at 0.9 0.5 RCIM_Cubic>
          ]}
        }}
       ScaleCurve=<1 1>
      )
     )
    )
   ) {isSetParameters=true}
   /SetVariables(Particles.SparkleFrontSize:Vector2f=&Multiply_Vector2D_ByFloat(Float=0.5, Vector2D=$Particles.SparkleSize)) {isSetParameters=true}
  ]
  rendererClasses [&NiagaraSpriteRendererProperties]
  renderers [
   {rendererClass=&NiagaraSpriteRendererProperties}
   {rendererClass=&NiagaraSpriteRendererProperties, rendererIndex=1}
   {rendererClass=&NiagaraSpriteRendererProperties, rendererIndex=2}
   {rendererClass=&NiagaraSpriteRendererProperties, rendererIndex=3}
  ]
 }
 MagicalDustFollowTrail {
  emitterSpawnScript [/SetVariables(Emitter.ParticlePerRibbon=20, Emitter.RibbonNumber=4) {isSetParameters=true}]
  emitterUpdateScript [
   &EmitterState(Life Cycle Mode=::NewEnumerator1, Loop Behavior=::NewEnumerator1, Loop Delay=1, Loop Duration Mode=::NewEnumerator1, UseLoopDelay=false)
   &SpawnRate(SpawnRate=50, Use Spawn Probability=false)
  ]
  particleSpawnScript [
   &PartitionParticles(Number Of Partitions=$Emitter.RibbonNumber, Write Partition to Ribbon ID=::NewEnumerator0)
   &InitializeParticle(
    Lifetime Max=1
    Lifetime Min=0.5
    Lifetime Mode=::NewEnumerator1
    Position Offset=<0 0 -22>
    Sprite Rotation Angle=100
    Sprite Rotation Mode=::NewEnumerator0
    Sprite Size Mode=::NewEnumerator1
    Uniform Sprite Size Max=&SelectFloatFromArray(Direct Array Index=$Particles.PartitionParticles.Partition, Float Selection Array=DataInterface() {propertyValues {FloatData [60,60,30,30]}})
    Uniform Sprite Size Min=&SelectFloatFromArray(Direct Array Index=$Particles.PartitionParticles.Partition, Float Selection Array=DataInterface() {propertyValues {FloatData [50,50,20,20]}})
    UsePositionOffset=true
    Write Material Random=true
   )
   &CameraOffset(Camera Offset Amount=25)
  ]
  particleUpdateScript [
   &ParticleState()
   &ScaleColor(
    Scale Alpha=&FloatFromCurve(
     FloatCurve=DataInterface() {propertyValues {Curve {keys [
         <at 0 0 RCIM_Cubic>
         {arriveTangent=2, interpMode=RCIM_Cubic, leaveTangent=0.0238, tangentMode=RCTM_Break, time=0.1, value=1}
         {interpMode=RCIM_Cubic, leaveTangent=-0.1316, tangentMode=RCTM_Break, time=0.9, value=1}
         {arriveTangent=-4.4762, tangentMode=RCTM_User, time=1, value=0}
        ]}}}
     Scale Curve=0.4
    )
    Scale RGB=&MakeVectorFromLinearColorRGB(Color=&ColorFromCurve(ColorCurve=DataInterface() {propertyValues {
        AlphaCurve {keys [<at 0 1 RCIM_Cubic>,<at 1 0 RCIM_Cubic>]}
        BlueCurve {keys [<at 0.4 1 RCIM_Cubic>,<at 0.65 0.4941 RCIM_Cubic>]}
        GreenCurve {keys [<at 0.4 1 RCIM_Cubic>,<at 0.65 0.2976 RCIM_Cubic>]}
        RedCurve {keys [<at 0.4 1 RCIM_Cubic>,<at 0.65 0.9922 RCIM_Cubic>]}
       }}))
    ScaleA=true
    ScaleRGB=true
   )
   &ScaleSpriteSize(
    Uniform Curve Scale=1.1
    Uniform Curve Sprite Scale=DataInterface() {propertyValues {Curve {keys [
        {interpMode=RCIM_Cubic, leaveTangent=12.2148, tangentMode=RCTM_Break, time=0, value=0}
        {arriveTangent=-1.0071, leaveTangent=-1.0071, time=0.1, value=0.8}
        {arriveTangent=-0.6384, interpMode=RCIM_Cubic, tangentMode=RCTM_Break, time=1, value=0.25}
       ]}}}
   )
   &self:FollowRibbon(
    Alpha:NiagaraFloat=&FloatFromCurve(FloatCurve=DataInterface() {propertyValues {Curve {keys [<at 0 0.3>,<at 1 1>]}}})
    ParticlePerRibbon:NiagaraInt32=$Emitter.ParticlePerRibbon
    RibbonNumber:NiagaraInt32=$Particles.PartitionParticles.Partition
    RibbonReader=DataInterface() {propertyValues {EmitterBinding {bindingMode=Other, emitterName=SmokeTrail}}}
   ) {module=FollowRibbon}
   /SetVariables(Particles.Position=&AddVectorToPosition(
     Position=$Output.FollowRibbon.RibbonPos
     Vector=&RandomVector(Vector Scale=&Lerp_Float(
       A=0
       B=&Lerp_Float(A=5, B=10, Alpha=$System.SmoothDiveAlpha)
       Alpha=&FloatFromCurve(FloatCurve=DataInterface() {propertyValues {Curve {keys [{interpMode=RCIM_Cubic, tangentMode=RCTM_User, time=0, value=0}, {interpMode=RCIM_Cubic, tangentMode=RCTM_User, time=1, value=1}]}}})
      ))
    )) {isSetParameters=true}
   /SetVariables(Particles.Velocity=$Output.FollowRibbon.DirectionVector) {isSetParameters=true}
   &DynamicMaterialParameters(
    Index 0 Param 1=0.35
    Module.Index0Float0Write=true
    Module.Index0Float1Write=true
    Module.Index0Float2Write=true
    Module.Index0Float3Write=true
   )
  ]
  rendererClasses [&NiagaraSpriteRendererProperties]
  renderers [{rendererClass=&NiagaraSpriteRendererProperties}]
 }
 SmokeParticles {
  emitterSpawnScript [/SetVariables(Emitter.ParticlePerRibbon=20, Emitter.RibbonNumber=4) {isSetParameters=true, module=SetVariables_#5}]
  emitterUpdateScript [
   &EmitterState(
    Inactive Response=::NewEnumerator1
    Life Cycle Mode=::NewEnumerator1
    Loop Behavior=::NewEnumerator1
    Loop Duration Mode=::NewEnumerator1
    UseLoopDelay=false
   )
   &SpawnBurst_Instantaneous(Spawn Count=12, Use Spawn Probability=false, UseLoopCountLimit=false) {enabled=false, module=SpawnBurst_Instantaneous001}
   &SpawnRate(SpawnRate=20, Use Spawn Probability=false) {enabled=false}
   &SpawnRate(SpawnGroup=1, SpawnRate=&Lerp_Float(A=50, B=60, Alpha=$System.ContrailDiveAlpha), Use Spawn Probability=false) {module=SpawnRate001}
  ]
  particleSpawnScript [
   &FN_SetSpawnGroup(Number of Spawn Groups=2) {enabled=false}
   /SetVariables(Particles.VisibilityTag=$Emitter.SpawnGroup) {enabled=false, isSetParameters=true, module=SetVariables_#6}
   &PartitionParticles(Number Of Partitions=$Emitter.RibbonNumber, Write Partition to Ribbon ID=::NewEnumerator0)
   &InitializeParticle(
    Color=<rgba 0.9294 0.9294 0.9294 1>
    Sprite Rotation Mode=::NewEnumerator2
    Sprite Size Mode=::NewEnumerator3
    Uniform Sprite Size=&RandomRangeFloat(
     Minimum=&SelectFloatFromArray(Direct Array Index=$Particles.PartitionParticles.Partition, Float Selection Array=DataInterface() {propertyValues {FloatData [50,50,20,20]}})
     Maximum=&SelectFloatFromArray(Direct Array Index=$Particles.PartitionParticles.Partition, Float Selection Array=DataInterface() {propertyValues {FloatData [60,60,30,30]}})
    )
    UsePositionOffset=false
    Write Material Random=true
   )
   &CameraOffset()
  ]
  particleUpdateScript [
   &ParticleState()
   &SpriteRotationRate(Rotation Rate=&RandomRangeFloat(Minimum=-50, Maximum=50))
   &ScaleSpriteSize(
    Scale Sprite Size Mode=::NewEnumerator0
    Uniform Scale Factor=&FloatFromCurve(FloatCurve=DataInterface() {propertyValues {Curve {keys [
         {interpMode=RCIM_Cubic, leaveTangent=12.2148, tangentMode=RCTM_Break, time=0, value=0}
         {arriveTangent=-0.5063, interpMode=RCIM_Cubic, leaveTangent=-0.5063, tangentMode=RCTM_User, time=0.1, value=0.9}
         {arriveTangent=-2.2021, interpMode=RCIM_Cubic, tangentMode=RCTM_Break, time=1, value=0}
        ]}}})
   )
   &SolveForcesAndVelocity(Clamp Velocity=false, Limit Acceleration=false)
   &self:FollowRibbon(
    Alpha:NiagaraFloat=&FloatFromCurve(FloatCurve=DataInterface() {propertyValues {Curve {keys [<at 0 0>,<at 1 1>]}}})
    ParticlePerRibbon:NiagaraInt32=$Emitter.ParticlePerRibbon
    RibbonNumber:NiagaraInt32=$Particles.PartitionParticles.Partition
    RibbonReader=DataInterface() {propertyValues {EmitterBinding {bindingMode=Other, emitterName=SmokeTrail}}}
   ) {module=FollowRibbon}
   /SetVariables(Particles.Position=&AddVectorToPosition(
     Position=$Output.FollowRibbon.RibbonPos
     Vector=&RandomVector(Vector Scale=&Lerp_Float(
       A=0
       B=&Lerp_Float(A=5, B=10, Alpha=$System.SmoothDiveAlpha)
       Alpha=&FloatFromCurve(FloatCurve=DataInterface() {propertyValues {Curve {keys [{interpMode=RCIM_Cubic, tangentMode=RCTM_User, time=0, value=0}, {interpMode=RCIM_Cubic, tangentMode=RCTM_User, time=1, value=1}]}}})
      ))
    )) {isSetParameters=true, module=SetVariables_#7}
   /SetVariables(Particles.Velocity=$Output.FollowRibbon.DirectionVector) {isSetParameters=true, module=SetVariables_#8}
   &DynamicMaterialParameters(
    Index 0 Param 1=&Add_Float(
     A=0
     B=&FloatFromCurve(
      FloatCurve=DataInterface() {propertyValues {Curve {keys [
          {arriveTangent=1.0192, interpMode=RCIM_Cubic, leaveTangent=1.0192, tangentMode=RCTM_User, time=0, value=1}
          {arriveTangent=0.2834, interpMode=RCIM_Cubic, leaveTangent=0.2834, tangentMode=RCTM_User, time=1, value=1.5}
         ]}}}
      CurveIndex=&Add_Float(A=&Multiply_Float(A=$Particles.Initial.MaterialRandom, B=0.5), B=$Particles.NormalizedAge)
      Scale Curve=0.4
     )
    )
    Index 0 Param 3=&FloatFromCurve(
     FloatCurve=DataInterface() {propertyValues {Curve {keys [{tangentMode=RCTM_User, time=0, value=1}, {tangentMode=RCTM_User, time=1, value=0}]}}}
     Scale Curve=0.05
    )
    Index 0 Param 4=&FloatFromCurve(FloatCurve=DataInterface() {propertyValues {Curve {keys [<at 0 1>,<at 1 0>]}}}, Scale Curve=0)
    Index 1 Param 1=1
    Index 1 Param 2=&FloatFromCurve(
     FloatCurve=DataInterface() {propertyValues {Curve {keys [<at 0 0.2>, {leaveTangent=0.9066, time=1, value=2}]}}}
     Scale Curve=&RandomRangeFloat(Minimum=0.3, Maximum=0.4)
    )
    Index 1 Param 3=&FloatFromCurve(
     FloatCurve=DataInterface() {propertyValues {Curve {keys [
         {interpMode=RCIM_Cubic, leaveTangent=6.3917, tangentMode=RCTM_Break, time=0, value=0.2}
         {arriveTangent=1.7549, interpMode=RCIM_Cubic, leaveTangent=10.6767, tangentMode=RCTM_User, time=0.25, value=1}
        ]}}}
     Scale Curve=0
    )
    Index 1 Param 4=&FloatFromCurve(
     FloatCurve=DataInterface() {propertyValues {Curve {keys [{arriveTangent=-0.136, interpMode=RCIM_Cubic, leaveTangent=-0.136, tangentMode=RCTM_User, time=0, value=1}, <at 1 0.95>]}}}
     Scale Curve=&RandomRangeFloat(Minimum=2.5, Maximum=3)
    )
    Index 2 Param 1=&FloatFromCurve(FloatCurve=DataInterface() {propertyValues {Curve {keys [{interpMode=RCIM_Cubic, leaveTangent=1.4301, tangentMode=RCTM_Break, time=0, value=0}, <at 1 1>]}}})
    Index 2 Param 2=&FloatFromCurve(
     FloatCurve=DataInterface() {propertyValues {Curve {keys [
         {arriveTangent=0.6777, interpMode=RCIM_Cubic, leaveTangent=0.6777, tangentMode=RCTM_User, time=0, value=0.1}
         {arriveTangent=0.0259, interpMode=RCIM_Cubic, leaveTangent=0.0259, tangentMode=RCTM_User, time=1, value=0.4}
        ]}}}
     Scale Curve=&RandomRangeFloat(Minimum=0.2, Maximum=0.5)
    )
    Index 2 Param 3=&FloatFromCurve(
     FloatCurve=DataInterface() {propertyValues {Curve {keys [
         {interpMode=RCIM_Cubic, leaveTangent=3.6087, tangentMode=RCTM_Break, time=0, value=1}
         {interpMode=RCIM_Cubic, leaveTangent=11.1233, tangentMode=RCTM_Break, time=1, value=2}
        ]}}}
     Scale Curve=3
    )
    Index 2 Param 4=&FloatFromCurve(FloatCurve=DataInterface() {propertyValues {Curve {keys [
         {arriveTangent=-0.4764, interpMode=RCIM_Cubic, leaveTangent=-0.4764, tangentMode=RCTM_User, time=0, value=1}
         {interpMode=RCIM_Cubic, leaveTangent=-1.2956, tangentMode=RCTM_Break, time=1, value=0.75}
        ]}}})
    Index 3 Linear Color:LinearColor=&ColorFromCurve(ColorCurve=DataInterface() {propertyValues {
       AlphaCurve {keys [<at 0 1 RCIM_Cubic>,<at 1 0 RCIM_Cubic>]}
       BlueCurve {keys [<at 0.1331 0.5823 RCIM_Cubic>,<at 0.9036 0.319 RCIM_Cubic>]}
       GreenCurve {keys [<at 0.1331 0.29 RCIM_Cubic>,<at 0.9036 0.1186 RCIM_Cubic>]}
       RedCurve {keys [<at 0.1331 1 RCIM_Cubic>,<at 0.9036 0.9294 RCIM_Cubic>]}
      }})
    Module.Index0Float0Write=true
    Module.Index0Float1Write=true
    Module.Index0Float2Write=true
    Module.Index0Float3Write=true
    Module.Index1Float0Write=true
    Module.Index1Float1Write=true
    Module.Index1Float2Write=true
    Module.Index1Float3Write=true
    Module.Index2Float0Write=true
    Module.Index2Float1Write=true
    Module.Index2Float2Write=true
    Module.Index2Float3Write=true
    Write Parameter Index 1=::NewEnumerator1
    Write Parameter Index 2=::NewEnumerator1
    Write Parameter Index 3=::NewEnumerator5
   )
   &ScaleColor(
    Scale Alpha=&FloatFromCurve(
     FloatCurve=DataInterface() {propertyValues {Curve {keys [
         {arriveTangent=2.6403, interpMode=RCIM_Cubic, leaveTangent=2.6403, tangentMode=RCTM_User, time=0, value=0.2}
         {arriveTangent=0.1466, leaveTangent=0.1466, time=0.5, value=1}
         {arriveTangent=-1.6, interpMode=RCIM_Cubic, leaveTangent=-1.6, time=0.8, value=1}
         {arriveTangent=-2.4925, interpMode=RCIM_Cubic, leaveTangent=-2.4925, tangentMode=RCTM_User, time=1, value=0.2}
        ]}}}
     Scale Curve=&Multiply_Float(
      A=1
      B=&CameraFade(
       Camera Query:NiagaraDataInterfaceCamera=DataInterface()
       Fade Offset:NiagaraFloat=0
       Fade Distance:NiagaraFloat=150
       Use Owner Position:NiagaraBool=true
      )
     )
    )
    Scale RGB=&MakeVectorFromLinearColorRGB(Color=&ColorFromCurve(ColorCurve=DataInterface() {propertyValues {
        AlphaCurve {keys [<at 0 1 RCIM_Cubic>]}
        BlueCurve {keys [<at 0 0.5385 RCIM_Cubic>,<at 1 0.4518 RCIM_Cubic>]}
        GreenCurve {keys [<at 0 0.1964 RCIM_Cubic>,<at 1 0.2268 RCIM_Cubic>]}
        RedCurve {keys [<at 0 0.9301 RCIM_Cubic>,<at 1 0.9922 RCIM_Cubic>]}
       }}))
    ScaleA=true
    ScaleRGB=true
   )
  ]
  rendererClasses [&NiagaraSpriteRendererProperties]
  renderers [
   {rendererClass=&NiagaraSpriteRendererProperties}
   {rendererClass=&NiagaraSpriteRendererProperties, rendererIndex=1}
  ]
 }
 SmokeParticlesHands {
  emitterSpawnScript [/SetVariables(Emitter.ParticlePerRibbon=20, Emitter.RibbonNumber=2) {isSetParameters=true, module=SetVariables_#5}]
  emitterUpdateScript [
   &EmitterState(
    Inactive Response=::NewEnumerator1
    Life Cycle Mode=::NewEnumerator1
    Loop Behavior=::NewEnumerator1
    Loop Duration Mode=::NewEnumerator1
    UseLoopDelay=false
   )
   &SpawnBurst_Instantaneous(Spawn Count=12, Use Spawn Probability=false, UseLoopCountLimit=false) {module=SpawnBurst_Instantaneous001}
   &SpawnRate(SpawnRate=3, Use Spawn Probability=false)
   &SpawnRate(SpawnGroup=1, SpawnRate=&Lerp_Float(A=50, B=60, Alpha=$System.ContrailDiveAlpha), Use Spawn Probability=false) {enabled=false, module=SpawnRate001}
  ]
  particleSpawnScript [
   &FN_SetSpawnGroup(Number of Spawn Groups=2) {enabled=false}
   /SetVariables(Particles.VisibilityTag=$Emitter.SpawnGroup) {enabled=false, isSetParameters=true, module=SetVariables_#6}
   &PartitionParticles(Number Of Partitions=$Emitter.RibbonNumber, Write Partition to Ribbon ID=::NewEnumerator0)
   &InitializeParticle(
    Color=<rgba 0.9294 0.9294 0.9294 1>
    Lifetime=3
    Sprite Rotation Mode=::NewEnumerator2
    Sprite Size Mode=::NewEnumerator3
    Uniform Sprite Size=&RandomRangeFloat(
     Minimum=&SelectFloatFromArray(Direct Array Index=$Particles.PartitionParticles.Partition, Float Selection Array=DataInterface() {propertyValues {FloatData [40,40,10,10]}})
     Maximum=&SelectFloatFromArray(Direct Array Index=$Particles.PartitionParticles.Partition, Float Selection Array=DataInterface() {propertyValues {FloatData [50,50,15,15]}})
    )
    UsePositionOffset=false
    Write Material Random=true
   )
   &CameraOffset(Camera Offset Amount=&FN_FrontEndFloatModifier(
     Gameplay=&SelectFloatFromArray(Direct Array Index=$Particles.PartitionParticles.Partition, Float Selection Array=DataInterface() {propertyValues {FloatData [25,25,0,0]}})
     Front End=&SelectFloatFromArray(Direct Array Index=$Particles.PartitionParticles.Partition, Float Selection Array=DataInterface() {propertyValues {FloatData [20,20,0,0]}})
    ))
  ]
  particleUpdateScript [
   &ParticleState()
   &SpriteRotationRate(Rotation Rate=&RandomRangeFloat(Minimum=-50, Maximum=50))
   &SolveForcesAndVelocity(Clamp Velocity=false, Limit Acceleration=false)
   &self:FollowRibbon(
    Alpha:NiagaraFloat=&FloatFromCurve(FloatCurve=DataInterface() {propertyValues {Curve {keys [<at 0 0>,<at 1 0.05>]}}})
    ParticlePerRibbon:NiagaraInt32=$Emitter.ParticlePerRibbon
    RibbonNumber:NiagaraInt32=$Particles.PartitionParticles.Partition
    RibbonReader=DataInterface() {propertyValues {EmitterBinding {bindingMode=Other, emitterName=SmokeTrail}}}
   ) {module=FollowRibbon}
   /SetVariables(Particles.Position=&AddVectorToPosition(
     Position=$Output.FollowRibbon.RibbonPos
     Vector=&RandomVector(Vector Scale=&Lerp_Float(
       A=&SelectFloatFromArray(Direct Array Index=$Particles.PartitionParticles.Partition, Float Selection Array=DataInterface() {propertyValues {FloatData [4,4,1,1]}})
       B=&Lerp_Float(A=5, B=10, Alpha=$System.SmoothDiveAlpha)
       Alpha=&FloatFromCurve(FloatCurve=DataInterface() {propertyValues {Curve {keys [{interpMode=RCIM_Cubic, tangentMode=RCTM_User, time=0, value=0}, {interpMode=RCIM_Cubic, tangentMode=RCTM_User, time=1, value=1}]}}})
      ))
    )) {isSetParameters=true, module=SetVariables_#7}
   /SetVariables(Particles.Velocity=$Output.FollowRibbon.DirectionVector) {isSetParameters=true, module=SetVariables_#8}
   &DynamicMaterialParameters(
    Index 0 Param 1=&Add_Float(
     A=0
     B=&FloatFromCurve(
      FloatCurve=DataInterface() {propertyValues {Curve {keys [
          {arriveTangent=1.0192, interpMode=RCIM_Cubic, leaveTangent=1.0192, tangentMode=RCTM_User, time=0, value=1}
          {arriveTangent=0.2834, interpMode=RCIM_Cubic, leaveTangent=0.2834, tangentMode=RCTM_User, time=1, value=1.5}
         ]}}}
      CurveIndex=&Add_Float(A=&Multiply_Float(A=$Particles.Initial.MaterialRandom, B=0.5), B=$Particles.NormalizedAge)
      Scale Curve=0.4
     )
    )
    Index 0 Param 3=&FloatFromCurve(
     FloatCurve=DataInterface() {propertyValues {Curve {keys [{tangentMode=RCTM_User, time=0, value=1}, {tangentMode=RCTM_User, time=1, value=0}]}}}
     Scale Curve=0.05
    )
    Index 0 Param 4=&FloatFromCurve(FloatCurve=DataInterface() {propertyValues {Curve {keys [<at 0 1>,<at 1 0>]}}}, Scale Curve=0.75)
    Index 1 Param 1=1
    Index 1 Param 2=&FloatFromCurve(
     FloatCurve=DataInterface() {propertyValues {Curve {keys [<at 0 0.2>, {leaveTangent=0.9066, time=1, value=2}]}}}
     Scale Curve=&RandomRangeFloat(Minimum=0.3, Maximum=0.4)
    )
    Index 1 Param 3=&FloatFromCurve(
     FloatCurve=DataInterface() {propertyValues {Curve {keys [
         {interpMode=RCIM_Cubic, leaveTangent=6.3917, tangentMode=RCTM_Break, time=0, value=0.2}
         {arriveTangent=1.7549, interpMode=RCIM_Cubic, leaveTangent=10.6767, tangentMode=RCTM_User, time=0.25, value=1}
        ]}}}
     Scale Curve=0
    )
    Index 1 Param 4=&FloatFromCurve(
     FloatCurve=DataInterface() {propertyValues {Curve {keys [{arriveTangent=-0.136, interpMode=RCIM_Cubic, leaveTangent=-0.136, tangentMode=RCTM_User, time=0, value=1}, <at 1 0.95>]}}}
     Scale Curve=&RandomRangeFloat(Minimum=2.5, Maximum=3)
    )
    Index 2 Param 1=&FloatFromCurve(FloatCurve=DataInterface() {propertyValues {Curve {keys [{interpMode=RCIM_Cubic, leaveTangent=1.4301, tangentMode=RCTM_Break, time=0, value=0}, <at 1 1>]}}})
    Index 2 Param 2=&FloatFromCurve(
     FloatCurve=DataInterface() {propertyValues {Curve {keys [
         {arriveTangent=0.6777, interpMode=RCIM_Cubic, leaveTangent=0.6777, tangentMode=RCTM_User, time=0, value=0.1}
         {arriveTangent=0.0259, interpMode=RCIM_Cubic, leaveTangent=0.0259, tangentMode=RCTM_User, time=1, value=0.4}
        ]}}}
     Scale Curve=&RandomRangeFloat(Minimum=0.2, Maximum=0.5)
    )
    Index 2 Param 3=&FloatFromCurve(
     FloatCurve=DataInterface() {propertyValues {Curve {keys [
         {interpMode=RCIM_Cubic, leaveTangent=3.6087, tangentMode=RCTM_Break, time=0, value=1}
         {interpMode=RCIM_Cubic, leaveTangent=11.1233, tangentMode=RCTM_Break, time=1, value=2}
        ]}}}
     Scale Curve=3
    )
    Index 2 Param 4=&FloatFromCurve(FloatCurve=DataInterface() {propertyValues {Curve {keys [
         {arriveTangent=-0.4764, interpMode=RCIM_Cubic, leaveTangent=-0.4764, tangentMode=RCTM_User, time=0, value=1}
         {interpMode=RCIM_Cubic, leaveTangent=-1.2956, tangentMode=RCTM_Break, time=1, value=0.75}
        ]}}})
    Index 3 Linear Color:LinearColor=<rgba 1 0.2505 0.4142 1>
    Module.Index0Float0Write=true
    Module.Index0Float1Write=true
    Module.Index0Float2Write=true
    Module.Index0Float3Write=true
    Module.Index1Float0Write=true
    Module.Index1Float1Write=true
    Module.Index1Float2Write=true
    Module.Index1Float3Write=true
    Module.Index2Float0Write=true
    Module.Index2Float1Write=true
    Module.Index2Float2Write=true
    Module.Index2Float3Write=true
    Write Parameter Index 1=::NewEnumerator1
    Write Parameter Index 2=::NewEnumerator1
    Write Parameter Index 3=::NewEnumerator5
   )
   &ScaleColor(
    Scale Alpha=&FloatFromCurve(
     FloatCurve=DataInterface() {propertyValues {Curve {keys [
         {arriveTangent=1.5531, interpMode=RCIM_Cubic, leaveTangent=1.5531, tangentMode=RCTM_User, time=0, value=0}
         {arriveTangent=-1.5789, interpMode=RCIM_Cubic, leaveTangent=-1.5789, tangentMode=RCTM_User, time=0.5, value=1}
         {arriveTangent=-2.9381, interpMode=RCIM_Cubic, leaveTangent=-2.9381, tangentMode=RCTM_User, time=1, value=0}
        ]}}}
     Scale Curve=2
    )
    Scale RGB=&MakeVectorFromLinearColorRGB(Color=<rgba 0.9294 0.364 0.4562 1>)
    ScaleA=true
    ScaleRGB=true
   )
  ]
  rendererClasses [&NiagaraSpriteRendererProperties]
  renderers [
   {rendererClass=&NiagaraSpriteRendererProperties}
   {rendererClass=&NiagaraSpriteRendererProperties, rendererIndex=1}
  ]
 }
 SmokeTrail {
  emitterSpawnScript [
   /SetVariables(Emitter.ParticlePerRibbon=20, Emitter.RibbonNumber=4) {isSetParameters=true, module=SetVariables_#5}
   /SetVariables(Emitter.rand:NiagaraFloat=&RandomRangeFloat(Minimum=0, Maximum=1)) {isSetParameters=true}
  ]
  emitterUpdateScript [
   &EmitterState(
    Inactive Response=::NewEnumerator1
    Life Cycle Mode=::NewEnumerator1
    Loop Behavior=::NewEnumerator1
    Loop Duration Mode=::NewEnumerator1
    UseLoopDelay=false
   )
   &SpawnBurst_Instantaneous(
    Spawn Count=&Multiply_Int(A=$Emitter.RibbonNumber, B=$Emitter.ParticlePerRibbon)
    Use Spawn Probability=false
    UseLoopCountLimit=false
   )
  ]
  particleSpawnScript [
   &PartitionParticles(
    Number Of Partitions=$Emitter.RibbonNumber
    Particles Per Partition:NiagaraInt32=$Emitter.ParticlePerRibbon
    Partition Mode:ENiagara_PartitionMode=::NewEnumerator1
    Write Partition to Ribbon ID=::NewEnumerator0
   )
   &InitializeParticle(
    Ribbon Width=17.5
    Ribbon Width Mode=::NewEnumerator1
    Sprite Size Mode=::NewEnumerator3
    Uniform Sprite Size=1
    UsePositionOffset=false
    Write Material Random=true
   )
   /SetVariables(Particles.RibbonU0Override=$Particles.PartitionParticles.Normalized Index In Partition) {isSetParameters=true}
  ]
  particleUpdateScript [
   &ParticleState(Kill Particles When Lifetime Has Elapsed=false)
   /SetVariables(Particles.RibbonLength:NiagaraFloat=&FN_FrontEndFloatModifier(
     Gameplay=&SelectFloatFromArray(
      Direct Array Index=$Particles.PartitionParticles.Partition
      Float Selection Array=DataInterface() {propertyValues {FloatData [150,150,50,50]}}
     )
     Front End=&Lerp_Float(
      A=&SelectFloatFromArray(
       Direct Array Index=$Particles.PartitionParticles.Partition
       Float Selection Array=DataInterface() {propertyValues {FloatData [150,150,50,50]}}
      )
      B=2
      Alpha=&FloatFromCurve(
       FloatCurve=DataInterface() {propertyValues {Curve {keys [
           {interpMode=RCIM_Cubic, tangentMode=RCTM_User, time=0, value=0}
           {interpMode=RCIM_Cubic, tangentMode=RCTM_User, time=0.45, value=1}
           {arriveTangent=0.0934, interpMode=RCIM_Cubic, leaveTangent=0.0934, tangentMode=RCTM_User, time=0.55, value=1}
           {interpMode=RCIM_Cubic, tangentMode=RCTM_User, time=1, value=0}
          ]}}}
       CurveIndex=$User.ParaGlideLeanAlpha
      )
     )
    )) {isSetParameters=true}
   /SetVariables(Particles.RibbonWidth=&Multiply_Float(
     A=$Particles.Initial.RibbonWidth
     B=&Lerp_Float(
      A=1
      B=&Lerp_Float(A=3.4, B=3.4, Alpha=$System.SmoothDiveAlpha)
      Alpha=&FloatFromCurve(
       FloatCurve=DataInterface() {propertyValues {Curve {keys [
           {arriveTangent=14.8039, interpMode=RCIM_Cubic, leaveTangent=14.8039, tangentMode=RCTM_User, time=0, value=0.1}
           {arriveTangent=-0.3275, interpMode=RCIM_Cubic, leaveTangent=-0.3275, tangentMode=RCTM_User, time=0.1, value=1}
           {tangentMode=RCTM_User, time=1, value=-0.4}
          ]}}}
       CurveIndex=$Particles.PartitionParticles.Normalized Index In Partition
       Scale Curve=&SelectFloatFromArray(Direct Array Index=$Particles.PartitionParticles.Partition, Float Selection Array=DataInterface() {propertyValues {FloatData [1,1,0.1,0.1]}})
      )
     )
    )) {isSetParameters=true}
   &SkeletalMeshLocation(
    Bone / Socket Index=$Particles.PartitionParticles.Partition
    Bone Sampling Mode=::NewEnumerator4
    MaskSpawnGroup=false
    Position Sampling=::NewEnumerator2
    Sampled Position Offset=&Lerp_Vector(
     A=&SelectVectorFromArray(
      Direct Array Index=$Particles.PartitionParticles.Partition
      Vector Selection Array=DataInterface() {propertyValues {FloatData [<0 0 -30>, <0 0 -30>, <2.5 -10 -30>, <-2.5 -10 -30>]}}
     )
     B=&SelectVectorFromArray(
      Direct Array Index=$Particles.PartitionParticles.Partition
      Vector Selection Array=DataInterface() {propertyValues {FloatData [<0 0 -30>, <0 0 -30>, <0 0 -30>, <0 0 -30>]}}
     )
     Alpha=$System.ContrailDiveAlpha
    )
    Sampled Position Offset Coordinate Space:ENiagaraSkelMeshTransforms=::NewEnumerator2
    Skeletal Mesh=$System.Skeletal Mesh
   )
   /SetVariables(Particles.GoalPos:NiagaraPosition=&AddVectorToPosition(
     Position=$Particles.SkeletalMeshLocation.SampledPosition
     Vector=&Multiply_VectorByFloat(
      Vector=&TransformVector(Vector=<0 0 1>, Destination Space=::Simulation)
      Float=&Lerp_Float(A=0, B=$Particles.RibbonLength, Alpha=$Particles.PartitionParticles.Normalized Index In Partition)
     )
    )) {isSetParameters=true}
   /SetVariables(Particles.Position=&Lerp_Vector(
     A=Linked() {linkedVariable {name=Particles.GoalPos, type=NiagaraPosition}}
     B=&Lerp_Vector(
      A=Linked() {linkedVariable {name=Particles.Position, type=NiagaraPosition}}
      B=Linked() {linkedVariable {name=Particles.GoalPos, type=NiagaraPosition}}
      Alpha=&Multiply_Float(
       A=0.05
       B=&FN_FrontEndFloatModifier(
        Gameplay=&SelectFloatFromArray(
         Direct Array Index=$Particles.PartitionParticles.Partition
         Float Selection Array=DataInterface() {propertyValues {FloatData [16.75,16.75,16.75,16.75]}}
        )
        Front End=&SelectFloatFromArray(
         Direct Array Index=$Particles.PartitionParticles.Partition
         Float Selection Array=DataInterface() {propertyValues {FloatData [16.75,16.75,16.75,16.75]}}
        )
       )
      )
     )
     Alpha=&FloatFromCurve(
      FloatCurve=DataInterface() {propertyValues {Curve {keys [{tangentMode=RCTM_User, time=0, value=0}, {tangentMode=RCTM_User, time=1, value=1}]}}}
      CurveIndex=$Particles.PartitionParticles.Normalized Index In Partition
     )
    )) {isSetParameters=true}
   /SetVariables(Particles.Position=&Add_Vector(
     A=&Add_Vector(
      A=Linked() {linkedVariable {name=Particles.Position, type=NiagaraPosition}}
      B=&FN_FrontEndVector3Modifier(Gameplay=<0 0 30>, Front End=<0 0 30>)
     )
     B=&Multiply_VectorByFloat(
      Vector=&Lerp_Vector(
       A=<0 0 0>
       B=&TransformVector(Vector=<5 0 0>, Destination Space=::Simulation, Apply Scale=true)
       Alpha=&FloatFromCurve(
        FloatCurve=DataInterface() {propertyValues {Curve {keys [{tangentMode=RCTM_User, time=0, value=0}, {tangentMode=RCTM_User, time=0.24, value=1}]}}}
        CurveIndex=$Particles.PartitionParticles.Normalized Index In Partition
       )
      )
      Float=&Sine(
       Normalized Angle=&Add_Float(
        A=&Add_Float(A=&Multiply_Float(A=$System.Age, B=1), B=$Emitter.rand)
        B=&Divide_Float(
         A=&FloatFromCurve(
          FloatCurve=DataInterface() {propertyValues {Curve {keys [{tangentMode=RCTM_User, time=0, value=0}, {tangentMode=RCTM_User, time=1, value=2}]}}}
          CurveIndex=$Particles.PartitionParticles.Normalized Index In Partition
          Scale Curve=$Particles.RibbonLength
         )
         B=&SelectFloatFromArray(
          Direct Array Index=$Particles.PartitionParticles.Partition
          Float Selection Array=DataInterface() {propertyValues {FloatData [150,-125,150,-125]}}
         )
        )
       )
       Period=1.25
       Scale=&FloatFromCurve(
        FloatCurve=DataInterface() {propertyValues {Curve {keys [
            {arriveTangent=0.1501, interpMode=RCIM_Cubic, leaveTangent=0.1501, tangentMode=RCTM_User, time=0, value=0}
            {arriveTangent=1.1179, interpMode=RCIM_Cubic, leaveTangent=1.1179, tangentMode=RCTM_User, time=0.1, value=0.5}
            {arriveTangent=0.0224, interpMode=RCIM_Cubic, leaveTangent=0.0224, tangentMode=RCTM_User, time=1, value=1}
           ]}}}
        CurveIndex=$Particles.PartitionParticles.Normalized Index In Partition
        Scale Curve=&SelectFloatFromArray(Direct Array Index=$Particles.PartitionParticles.Partition, Float Selection Array=DataInterface() {propertyValues {FloatData [1.5,1.5,1,1]}})
       )
       Bias=0
      )
     )
    )) {isSetParameters=true, module=SetVariables_#9}
   /SetVariables(Particles.Position=&AddVectorToPosition(
     Position=$Particles.Position
     Vector=&FN_FrontEndVector3Modifier(
      Gameplay=&Multiply_Vector(
       A=&TransformVector(Vector=$Engine.Owner.Velocity, Source Space=::World, Destination Space=::Simulation)
       B=&Multiply_Vector(
        A=&VectorFromFloat(Value=&Multiply_Float(
          A=&Lerp_Float(
           A=&SelectFloatFromArray(Direct Array Index=$Particles.PartitionParticles.Partition, Float Selection Array=DataInterface() {propertyValues {FloatData [1,1,1,1]}})
           B=&SelectFloatFromArray(Direct Array Index=$Particles.PartitionParticles.Partition, Float Selection Array=DataInterface() {propertyValues {FloatData [1,1,0,0]}})
           Alpha=$System.Dive
          )
          B=&FloatFromCurve(
           FloatCurve=DataInterface() {propertyValues {Curve {keys [{tangentMode=RCTM_User, time=0, value=0}, {tangentMode=RCTM_User, time=1, value=1}]}}}
           CurveIndex=$Particles.PartitionParticles.Normalized Index In Partition
          )
         ))
        B=<-0.05 -0.05 0>
       )
      )
      Front End=&Lerp_Vector(
       A=<0 0 0>
       B=&TransformVector(
        Vector=&VectorFromCurve(
         VectorCurve=DataInterface() {propertyValues {
           XCurve {keys [{tangentMode=RCTM_User, time=0, value=0}, {tangentMode=RCTM_User, time=1, value=1}]}
           YCurve {keys [{tangentMode=RCTM_User, time=0, value=0}, {tangentMode=RCTM_User, time=1, value=1}]}
           ZCurve {keys [{tangentMode=RCTM_User, time=0, value=0}, {tangentMode=RCTM_User, time=1, value=1}]}
          }}
         CurveIndex=$Particles.PartitionParticles.Normalized Index In Partition
         Scale Curve=&SelectVectorFromArray(
          Direct Array Index=$Particles.PartitionParticles.Partition
          Vector Selection Array=DataInterface() {propertyValues {FloatData [<0 -200 0>, <0 -200 1>, <0 -100 0>, <0 -100 0>]}}
         )
        )
        Destination Space=::Simulation
       )
       Alpha=$System.Dive
      )
     )
    )) {isSetParameters=true}
   &DynamicMaterialParameters(
    Index 0 Param 1=&SelectFloatFromArray(
     Direct Array Index=$Particles.PartitionParticles.Partition
     Float Selection Array=DataInterface() {propertyValues {FloatData [0.2,0.5,0.6,0.2]}}
    )
    Index 0 Param 4=1
    Index 1 Param 3=&FloatFromCurve(
     FloatCurve=DataInterface() {propertyValues {Curve {keys [<at 0.25 0>,<at 1 1>]}}}
     CurveIndex=$Particles.PartitionParticles.Normalized Index In Partition
     Scale Curve=2.5
    )
    Module.Index0Float0Write=true
    Module.Index0Float1Write=false
    Module.Index0Float2Write=false
    Module.Index0Float3Write=true
    Module.Index1Float0Write=false
    Module.Index1Float1Write=false
    Module.Index1Float2Write=true
    Module.Index1Float3Write=false
    Module.Index2Float0Write=true
    Module.Index2Float1Write=true
    Module.Index2Float2Write=true
    Module.Index2Float3Write=true
    Write Parameter Index 1=::NewEnumerator1
    Write Parameter Index 2=::NewEnumerator1
   )
   &InheritSourceMovement(
    Applied Position Scale=&Multiply_Vector(
     A=&VectorFromFloat(Value=&Multiply_Float(
       A=&Lerp_Float(
        A=&SelectFloatFromArray(Direct Array Index=$Particles.PartitionParticles.Partition, Float Selection Array=DataInterface() {propertyValues {FloatData [1,1,1,1]}})
        B=&SelectFloatFromArray(Direct Array Index=$Particles.PartitionParticles.Partition, Float Selection Array=DataInterface() {propertyValues {FloatData [-2,-2,0,0]}})
        Alpha=$System.Dive
       )
       B=&FloatFromCurve(
        FloatCurve=DataInterface() {propertyValues {Curve {keys [{tangentMode=RCTM_User, time=0, value=0}, {tangentMode=RCTM_User, time=1, value=1}]}}}
        CurveIndex=$Particles.PartitionParticles.Normalized Index In Partition
       )
      ))
     B=<1 1 0>
    )
    ApplyForce=false
    ApplyPosition=true
    ApplyPositionFalloff=false
   ) {enabled=false}
   &SolveForcesAndVelocity(Clamp Velocity=false, Limit Acceleration=false)
   &ScaleColor(
    Scale Alpha=&Multiply_Float(
     A=&FloatFromCurve(
      FloatCurve=DataInterface() {propertyValues {Curve {keys [
          {interpMode=RCIM_Cubic, tangentMode=RCTM_User, time=0.5, value=1}
          {arriveTangent=-7.3726, interpMode=RCIM_Cubic, leaveTangent=-7.3726, tangentMode=RCTM_User, time=0.75, value=0}
         ]}}}
      CurveIndex=$Particles.PartitionParticles.Normalized Index In Partition
      Scale Curve=&Multiply_Float(
       A=1
       B=&CameraFade(
        Camera Query:NiagaraDataInterfaceCamera=DataInterface()
        Fade Offset:NiagaraFloat=0
        Fade Distance:NiagaraFloat=150
        Use Owner Position:NiagaraBool=true
       )
      )
     )
     B=&FN_FrontEndFloatModifier(
      Gameplay=1
      Front End=&FloatFromCurve(
       FloatCurve=DataInterface() {propertyValues {Curve {keys [
           <at 0 0 RCIM_Cubic>
           {interpMode=RCIM_Cubic, tangentMode=RCTM_User, time=0.2, value=0.95}
           {interpMode=RCIM_Cubic, tangentMode=RCTM_User, time=0.45, value=0}
           {arriveTangent=0.0934, interpMode=RCIM_Cubic, leaveTangent=0.0934, tangentMode=RCTM_User, time=0.55, value=0}
           {interpMode=RCIM_Cubic, tangentMode=RCTM_User, time=1, value=1}
          ]}}}
       CurveIndex=$User.ParaGlideLeanAlpha
      )
     )
    )
    Scale RGB=&MakeVectorFromLinearColorRGB(Color=&ColorFromCurve(
      ColorCurve=DataInterface() {propertyValues {
        AlphaCurve {keys [<at 0 1 RCIM_Cubic>]}
        BlueCurve {keys [<at 0 0.5416 RCIM_Cubic>,<at 1 0.4348 RCIM_Cubic>]}
        GreenCurve {keys [<at 0 0.3885 RCIM_Cubic>,<at 1 0.266 RCIM_Cubic>]}
        RedCurve {keys [<at 0 0.9294 RCIM_Cubic>,<at 1 0.9911 RCIM_Cubic>]}
       }}
      CurveIndex=$Particles.PartitionParticles.Normalized Index In Partition
     ))
    ScaleA=true
    ScaleRGB=true
   ) {module=ScaleColor001}
  ]
  rendererClasses [&NiagaraRibbonRendererProperties]
  renderers [{rendererClass=&NiagaraRibbonRendererProperties}]
 }
}
summary {
 systemName=NS_Contrail_DimeBlanket
 emitters [
  {emitterName=FabricRibbon, rendererClasses [&NiagaraRibbonRendererProperties]}
  {emitterName=MagicalDustFollowTrail, rendererClasses [&NiagaraSpriteRendererProperties]}
  {emitterName=Kisses_Sparkles, rendererClasses [&NiagaraSpriteRendererProperties]}
  {emitterName=Hearts, rendererClasses [&NiagaraSpriteRendererProperties]}
  {emitterName=SmokeTrail, rendererClasses [&NiagaraRibbonRendererProperties]}
  {emitterName=SmokeParticles, rendererClasses [&NiagaraSpriteRendererProperties]}
  {emitterName=SmokeParticlesHands, rendererClasses [&NiagaraSpriteRendererProperties]}
 ]
 userVariables [
  User.bIsFrontEnd {type=NiagaraBool}
  User.bIsFrontEndPreview {type=NiagaraBool}
  User.ForwardDot {type=NiagaraFloat, defaultValue=0, description=""}
  User.IsSkydiving {type=NiagaraBool}
  User.Lifetime_Scale {type=NiagaraFloat, defaultValue=0, description=""}
  User.ParaGlideLeanAlpha {type=NiagaraFloat, defaultValue=0, description=""}
  User.RibbonIDOffset {type=NiagaraInt32}
  User.Variant {type=NiagaraFloat, defaultValue=0, description=""}
 ]
}
